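# Physics asset for the SherWolf Gnash back bling cape, FrontEnd variant by name, pairing capsule bodies on torso, limbs and backpack with cape bones.
# Four six-segment cape chains (left, right, mid-left, mid-right) hang from small MSTR bodies constrained to the clavicles and spine_05.
# Body and MSTR joints swing and twist freely.
# Cape segments have tuned swing limits and zero twist, with outer chains mostly bending on one axis.
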PhysicsAsset /BRCosmetics/Characters/Player/Male/Medium/Bodies/M_MED_SherWolf_Gnash/Meshes/M_MED_SherWolf_GnashBackBling_CapeFrontEnd_Physics1
bodies {
 backpack_collision [{
   shapeType=Capsule
   length=18.5726
   radius=25.3425
   shapeName=backpack_collision_capsule
   center=<-0.0001 -25.2919 6.6482>
   rotation=<rot -90 179.9987 -179.9989>
  }]
 calf_l [{shapeType=Capsule, length=40.679, radius=9.6561, shapeName=calf_l_capsule, center=<-18.4993 0 0>, rotation=<rot 90 0 0>}]
 calf_r [{shapeType=Capsule, length=40.679, radius=9.6561, shapeName=calf_r_capsule, center=<18.4993 0 0>, rotation=<rot 90 0 0>}]
 cape_01_l [{shapeType=Capsule, length=13.0431, radius=6.9275, shapeName=cape_01_l_capsule, center=<11.9983 0 0>, rotation=<rot 90 153.435 -116.565>}]
 cape_01_r [{shapeType=Capsule, length=9.1914, radius=5.5845, shapeName=cape_01_r_capsule, center=<7.8254 0 2.0092>, rotation=<rot 86.9366 -180 -180>}]
 cape_02_l [
  {shapeType=Capsule, length=9.564, radius=7.6988, shapeName=cape_02_l_capsule, center=<18.223 0 0>, rotation=<rot 0 0 90>}
  {shapeType=Capsule, length=14.1165, radius=6.0281, shapeName=cape_02_l_capsule, center=<6.364 0 0>, rotation=<rot 0 0 90>}
 ]
 cape_02_r [
  {shapeType=Capsule, length=16.75, radius=7.25, shapeName=cape_02_r_capsule, center=<7.9328 0 5.4754>, rotation=<rot 90 -90 0>}
  {shapeType=Capsule, length=16.75, radius=7.25, shapeName=cape_02_r_capsule, center=<20.1873 0 3.9857>, rotation=<rot 90 -90 0>}
 ]
 cape_03_l [{shapeType=Capsule, length=9.9831, radius=7.3946, shapeName=cape_03_l_capsule, center=<11.0576 0 0>, rotation=<rot 90 0 90>}]
 cape_03_r [{shapeType=Capsule, length=13.0209, radius=7.7929, shapeName=cape_03_r_capsule, center=<8 0 0>, rotation=<rot 0 0 90>}]
 cape_04_l [{shapeType=Capsule, length=7.1428, radius=7.4916, shapeName=cape_04_l_capsule, center=<9.4987 0 0>, rotation=<rot 0 0 90>}]
 cape_04_r [{shapeType=Capsule, length=15.1071, radius=6.7397, shapeName=cape_04_r_capsule, center=<8 -9 2.5>, rotation=<rot 90 0 0>}]
 cape_05_l [{shapeType=Capsule, length=5.4661, radius=7.5719, shapeName=cape_05_l_capsule, center=<7.655 0 0>, rotation=<rot 90 0 90>}]
 cape_05_r [{shapeType=Capsule, length=20.3052, radius=6.4635, shapeName=cape_05_r_capsule, center=<8 0 0>, rotation=<rot 0 0 90>}]
 cape_06_l [{shapeType=Capsule, length=7.5137, radius=7.0756, shapeName=cape_06_l_capsule, center=<7.655 0 0>, rotation=<rot 0 0 90>}]
 cape_06_r [{shapeType=Capsule, length=20.9681, radius=5.7645, shapeName=cape_06_r_capsule, center=<8 0 0>, rotation=<rot 0 0 90>}]
 cape_l_MSTR [{shapeType=Capsule, length=1, radius=1, shapeName=cape_l_MSTR_capsule, rotation=<rot 0 -90 0>}]
 cape_mid_01_l [{shapeType=Capsule, length=15.772, radius=4.5132, shapeName=cape_mid_01_l_capsule, center=<10.9091 0 5.8>, rotation=<rot 82.4316 -180 -180>}]
 cape_mid_01_r [{shapeType=Capsule, length=15.772, radius=4.5132, shapeName=cape_mid_01_r_capsule, center=<10.9091 0 5.8>, rotation=<rot 82.4316 180 180>}]
 cape_mid_02_l [{shapeType=Capsule, length=15.2976, radius=5.525, shapeName=cape_mid_02_l_capsule, center=<15.4392 0 0>, rotation=<rot 90 0 0>}]
 cape_mid_02_r [{shapeType=Capsule, length=15.2976, radius=5.525, shapeName=cape_mid_02_r_capsule, center=<15.4392 0 0>, rotation=<rot 90 0 0>}]
 cape_mid_03_l [{shapeType=Capsule, length=13.6579, radius=8.025, shapeName=cape_mid_03_l_capsule, center=<11.0576 0 0>, rotation=<rot 0 0 90>}]
 cape_mid_03_r [{shapeType=Capsule, length=13.6579, radius=8.025, shapeName=cape_mid_03_r_capsule, center=<11.0576 0 0>, rotation=<rot 0 0 90>}]
 cape_mid_04_l [{shapeType=Capsule, length=15.5186, radius=8.357, shapeName=cape_mid_04_l_capsule, center=<9.4987 0 4.2>, rotation=<rot 90 180 -90>}]
 cape_mid_04_r [{shapeType=Capsule, length=13.4271, radius=8.357, shapeName=cape_mid_04_r_capsule, center=<9.4987 0 4.2>, rotation=<rot 90 180 -90>}]
 cape_mid_05_l [{shapeType=Capsule, length=15.8695, radius=7.5953, shapeName=cape_mid_05_l_capsule, center=<7.655 0 5>, rotation=<rot 90 153.435 -116.565>}]
 cape_mid_05_r [{shapeType=Capsule, length=15.377, radius=8.0879, shapeName=cape_mid_05_r_capsule, center=<7.655 0 5>, rotation=<rot 90 153.435 -116.565>}]
 cape_mid_06_l [{shapeType=Capsule, length=17.2964, radius=7.7197, shapeName=cape_mid_06_l_capsule, center=<9.6942 0 3>, rotation=<rot 0 0 90>}]
 cape_mid_06_r [{shapeType=Capsule, length=17.3095, radius=7.7065, shapeName=cape_mid_06_r_capsule, center=<9.6942 0 3>, rotation=<rot 0 0 90>}]
 cape_mid_l_MSTR [{shapeType=Capsule, length=1, radius=1, shapeName=cape_mid_l_MSTR_capsule, rotation=<rot 0 -90 0>}]
 cape_mid_r_MSTR [{shapeType=Capsule, length=1, radius=1, shapeName=cape_mid_r_MSTR_capsule, rotation=<rot 0 -90 0>}]
 cape_r_MSTR [{shapeType=Capsule, length=1, radius=1, shapeName=cape_r_MSTR_capsule, rotation=<rot 0 -90 0>}]
 clavicle_l [{shapeType=Capsule, length=12.0737, radius=7.5263, shapeName=clavicle_l_capsule, center=<9.3 0 0>, rotation=<rot 81 0 0>}]
 clavicle_r [{shapeType=Capsule, length=12.0737, radius=7.5263, shapeName=clavicle_r_capsule, center=<-9.3 0 0>, rotation=<rot 81 0 0>}]
 foot_l [
  {shapeType=Capsule, length=17.8017, radius=7, shapeName=foot_l_capsule, center=<-4.0851 7.2082 0>, rotation=<rot 90 180 -90>}
  {shapeType=Capsule, length=12.1748, radius=9.0432, shapeName=foot_l_capsule, center=<0 -1.6375 0>}
 ]
 foot_r [
  {shapeType=Capsule, length=16.3613, radius=7, shapeName=foot_r_capsule, center=<5.1029 -7.0735 0>, rotation=<rot 90 180 -90>}
  {shapeType=Capsule, length=12.1748, radius=9.0432, shapeName=foot_r_capsule, center=<0 2.2446 0>}
 ]
 lowerarm_l [{shapeType=Capsule, length=26.3904, radius=6.1093, shapeName=lowerarm_l_capsule, center=<10.6887 0 0>, rotation=<rot 90 0 0>}]
 lowerarm_r [{shapeType=Capsule, length=18.7314, radius=5.6284, shapeName=lowerarm_r_capsule, center=<-18.4993 0 0>, rotation=<rot 90 0 0>}]
 pelvis [
  {shapeType=Capsule, length=11.546, radius=12.8138, shapeName=pelvis_capsule}
  {shapeType=Capsule, length=23, radius=15, shapeName=pelvis_capsule, center=<0 6.0445 0>}
 ]
 spine_02 [
  {shapeType=Capsule, length=14.1596, radius=10.2002, shapeName=spine_02_capsule}
  {shapeType=Capsule, length=20.5, radius=15, shapeName=spine_02_capsule, center=<0 4.6445 0>}
 ]
 spine_03 [{shapeType=Capsule, length=13.5763, radius=11.3042, shapeName=spine_03_capsule}]
 spine_04 [{shapeType=Capsule, length=18.2137, radius=13.6257, shapeName=spine_04_capsule}]
 spine_05 [{shapeType=Capsule, length=26.0542, radius=11.294, shapeName=spine_05_capsule}]
 thigh_l [{shapeType=Capsule, length=30.1221, radius=11.1367, shapeName=thigh_l_capsule, center=<-18.4993 0 0>, rotation=<rot 90 0 0>}]
 thigh_r [{shapeType=Capsule, length=30.1221, radius=11.1367, shapeName=thigh_r_capsule, center=<18.4993 0 0>, rotation=<rot 90 0 0>}]
 upperarm_l [{shapeType=Capsule, length=20.0045, radius=7.0339, shapeName=upperarm_l_capsule, center=<14.4803 0 0>, rotation=<rot 90 0 0>}]
 upperarm_r [{shapeType=Capsule, length=18.1862, radius=6.1736, shapeName=upperarm_r_capsule, center=<-18.4993 -1.3147 0>, rotation=<rot 90 5.3953 0>}]
}
constraints [
 {
  bone1Name=spine_05
  bone2Name=spine_04
  swing1LimitDegrees=45
  swing1Motion=Free
  swing2LimitDegrees=45
  swing2Motion=Free
  twistLimitDegrees=45
  twistMotion=Free
 }
 {
  bone1Name=clavicle_l
  bone2Name=spine_05
  swing1LimitDegrees=45
  swing1Motion=Free
  swing2LimitDegrees=45
  swing2Motion=Free
  twistLimitDegrees=45
  twistMotion=Free
 }
 {
  bone1Name=spine_04
  bone2Name=spine_03
  swing1LimitDegrees=45
  swing1Motion=Free
  swing2LimitDegrees=45
  swing2Motion=Free
  twistLimitDegrees=45
  twistMotion=Free
 }
 {
  bone1Name=upperarm_l
  bone2Name=clavicle_l
  swing1LimitDegrees=45
  swing1Motion=Free
  swing2LimitDegrees=45
  swing2Motion=Free
  twistLimitDegrees=45
  twistMotion=Free
 }
 {
  bone1Name=spine_02
  bone2Name=pelvis
  swing1LimitDegrees=45
  swing1Motion=Free
  swing2LimitDegrees=45
  swing2Motion=Free
  twistLimitDegrees=45
  twistMotion=Free
 }
 {
  bone1Name=lowerarm_l
  bone2Name=upperarm_l
  swing1LimitDegrees=45
  swing1Motion=Free
  swing2LimitDegrees=45
  swing2Motion=Free
  twistLimitDegrees=45
  twistMotion=Free
 }
 {
  bone1Name=clavicle_r
  bone2Name=spine_05
  swing1LimitDegrees=45
  swing1Motion=Free
  swing2LimitDegrees=45
  swing2Motion=Free
  twistLimitDegrees=45
  twistMotion=Free
 }
 {
  bone1Name=upperarm_r
  bone2Name=clavicle_r
  swing1LimitDegrees=45
  swing1Motion=Free
  swing2LimitDegrees=45
  swing2Motion=Free
  twistLimitDegrees=45
  twistMotion=Free
 }
 {
  bone1Name=lowerarm_r
  bone2Name=upperarm_r
  swing1LimitDegrees=45
  swing1Motion=Free
  swing2LimitDegrees=45
  swing2Motion=Free
  twistLimitDegrees=45
  twistMotion=Free
 }
 {
  bone1Name=thigh_l
  bone2Name=pelvis
  swing1LimitDegrees=45
  swing1Motion=Free
  swing2LimitDegrees=45
  swing2Motion=Free
  twistLimitDegrees=45
  twistMotion=Free
 }
 {
  bone1Name=calf_l
  bone2Name=thigh_l
  swing1LimitDegrees=45
  swing1Motion=Free
  swing2LimitDegrees=45
  swing2Motion=Free
  twistLimitDegrees=45
  twistMotion=Free
 }
 {
  bone1Name=thigh_r
  bone2Name=pelvis
  swing1LimitDegrees=45
  swing1Motion=Free
  swing2LimitDegrees=45
  swing2Motion=Free
  twistLimitDegrees=45
  twistMotion=Free
 }
 {
  bone1Name=calf_r
  bone2Name=thigh_r
  swing1LimitDegrees=45
  swing1Motion=Free
  swing2LimitDegrees=45
  swing2Motion=Free
  twistLimitDegrees=45
  twistMotion=Free
 }
 {
  bone1Name=spine_03
  bone2Name=spine_02
  swing1LimitDegrees=45
  swing1Motion=Free
  swing2LimitDegrees=45
  swing2Motion=Free
  twistLimitDegrees=45
  twistMotion=Free
 }
 {
  bone1Name=cape_l_MSTR
  bone2Name=clavicle_l
  swing1LimitDegrees=45
  swing1Motion=Free
  swing2LimitDegrees=45
  swing2Motion=Free
  twistLimitDegrees=45
  twistMotion=Free
 }
 {bone1Name=cape_01_l, bone2Name=cape_l_MSTR, swing1LimitDegrees=30, swing2LimitDegrees=20, twistLimitDegrees=0}
 {bone1Name=cape_02_l, bone2Name=cape_01_l, swing1LimitDegrees=0, swing2LimitDegrees=25, twistLimitDegrees=0}
 {bone1Name=cape_03_l, bone2Name=cape_02_l, swing1LimitDegrees=0, swing2LimitDegrees=20, twistLimitDegrees=0}
 {bone1Name=cape_04_l, bone2Name=cape_03_l, swing1LimitDegrees=0, swing2LimitDegrees=20, twistLimitDegrees=0}
 {bone1Name=cape_05_l, bone2Name=cape_04_l, swing1LimitDegrees=0, swing2LimitDegrees=20, twistLimitDegrees=0}
 {bone1Name=cape_06_l, bone2Name=cape_05_l, swing1LimitDegrees=0, swing2LimitDegrees=20, twistLimitDegrees=0}
 {
  bone1Name=cape_r_MSTR
  bone2Name=clavicle_r
  swing1LimitDegrees=45
  swing1Motion=Free
  swing2LimitDegrees=45
  swing2Motion=Free
  twistLimitDegrees=45
  twistMotion=Free
 }
 {
  bone1Name=cape_mid_l_MSTR
  bone2Name=spine_05
  swing1LimitDegrees=45
  swing1Motion=Free
  swing2LimitDegrees=45
  swing2Motion=Free
  twistLimitDegrees=45
  twistMotion=Free
 }
 {bone1Name=cape_mid_01_l, bone2Name=cape_mid_l_MSTR, swing1LimitDegrees=5, swing2LimitDegrees=30, twistLimitDegrees=0}
 {bone1Name=cape_mid_02_l, bone2Name=cape_mid_01_l, swing1LimitDegrees=10, swing2LimitDegrees=22, twistLimitDegrees=0}
 {bone1Name=cape_mid_03_l, bone2Name=cape_mid_02_l, swing1LimitDegrees=10, swing2LimitDegrees=24, twistLimitDegrees=0}
 {bone1Name=cape_mid_04_l, bone2Name=cape_mid_03_l, swing1LimitDegrees=10, swing2LimitDegrees=30, twistLimitDegrees=0}
 {bone1Name=cape_mid_05_l, bone2Name=cape_mid_04_l, swing1LimitDegrees=20, swing2LimitDegrees=20, twistLimitDegrees=0}
 {bone1Name=cape_mid_06_l, bone2Name=cape_mid_05_l, swing1LimitDegrees=20, swing2LimitDegrees=20, twistLimitDegrees=0}
 {
  bone1Name=cape_mid_r_MSTR
  bone2Name=spine_05
  swing1LimitDegrees=45
  swing1Motion=Free
  swing2LimitDegrees=45
  swing2Motion=Free
  twistLimitDegrees=45
  twistMotion=Free
 }
 {bone1Name=cape_mid_01_r, bone2Name=cape_mid_r_MSTR, swing1LimitDegrees=5, swing2LimitDegrees=30, twistLimitDegrees=0}
 {bone1Name=cape_mid_02_r, bone2Name=cape_mid_01_r, swing1LimitDegrees=10, swing2LimitDegrees=22, twistLimitDegrees=0}
 {bone1Name=cape_mid_03_r, bone2Name=cape_mid_02_r, swing1LimitDegrees=10, swing2LimitDegrees=24, twistLimitDegrees=0}
 {bone1Name=cape_mid_04_r, bone2Name=cape_mid_03_r, swing1LimitDegrees=10, swing2LimitDegrees=30, twistLimitDegrees=0}
 {bone1Name=cape_mid_05_r, bone2Name=cape_mid_04_r, swing1LimitDegrees=20, swing2LimitDegrees=20, twistLimitDegrees=0}
 {bone1Name=cape_mid_06_r, bone2Name=cape_mid_05_r, swing1LimitDegrees=20, swing2LimitDegrees=20, twistLimitDegrees=0}
 {
  bone1Name=foot_l
  bone2Name=calf_l
  swing1LimitDegrees=45
  swing1Motion=Free
  swing2LimitDegrees=45
  swing2Motion=Free
  twistLimitDegrees=45
  twistMotion=Free
 }
 {
  bone1Name=foot_r
  bone2Name=calf_r
  swing1LimitDegrees=45
  swing1Motion=Free
  swing2LimitDegrees=45
  swing2Motion=Free
  twistLimitDegrees=45
  twistMotion=Free
 }
 {bone1Name=cape_01_r, bone2Name=cape_r_MSTR, swing1LimitDegrees=10, swing2LimitDegrees=20, twistLimitDegrees=0}
 {bone1Name=cape_02_r, bone2Name=cape_01_r, swing1LimitDegrees=0, swing2LimitDegrees=20, twistLimitDegrees=0}
 {bone1Name=cape_03_r, bone2Name=cape_02_r, swing1LimitDegrees=0, swing2LimitDegrees=30, twistLimitDegrees=0}
 {bone1Name=cape_04_r, bone2Name=cape_03_r, swing1LimitDegrees=0, swing2LimitDegrees=30, twistLimitDegrees=0}
 {bone1Name=cape_05_r, bone2Name=cape_04_r, swing1LimitDegrees=0, swing2LimitDegrees=30, twistLimitDegrees=0}
 {bone1Name=cape_06_r, bone2Name=cape_05_r, swing1LimitDegrees=0, swing2LimitDegrees=20, twistLimitDegrees=0}
]
