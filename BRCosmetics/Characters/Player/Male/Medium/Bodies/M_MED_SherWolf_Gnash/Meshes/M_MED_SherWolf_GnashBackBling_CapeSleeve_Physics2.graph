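# Physics asset for the left and right cape sleeves of the SherWolf Gnash back bling on the medium male body.
# Each side chains four sleeve capsules from a tiny root capsule, with all angular motion locked at the first joint, swing2 and twist locked at the second, and none locked at the last.
# Unconstrained capsules on the forearms and dedicated sleeve_collision bones provide additional collision volumes.
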PhysicsAsset /BRCosmetics/Characters/Player/Male/Medium/Bodies/M_MED_SherWolf_Gnash/Meshes/M_MED_SherWolf_GnashBackBling_CapeSleeve_Physics2
bodies {
 lowerarm_l [{shapeType=Capsule, length=19.6449, radius=6.8146, shapeName=lowerarm_l_capsule, center=<8.7884 0 0.053>, rotation=<rot 0 -90 -90.3456>}]
 lowerarm_r [{
   shapeType=Capsule
   length=18.7576
   radius=7.4342
   shapeName=lowerarm_r_capsule
   center=<-13.3632 0 -0.0806>
   rotation=<rot 0 -89.9999 -90.3456>
  }]
 sleeve_01_l [{shapeType=Capsule, length=0.505, radius=0.505, shapeName=sleeve_01_l_capsule, rotation=<rot 0 -90 0>}]
 sleeve_01_r [{shapeType=Capsule, length=0.505, radius=0.505, shapeName=sleeve_01_r_capsule, rotation=<rot 0 -90 0>}]
 sleeve_02_l [{
   shapeType=Capsule
   length=18.9397
   radius=8.7668
   shapeName=sleeve_02_r_capsule
   center=<10.8215 1.0806 7.5982>
   rotation=<rot 6.3022 -93.0058 -93.6798>
  }]
 sleeve_02_r [{
   shapeType=Capsule
   length=12.4769
   radius=6.8057
   shapeName=sleeve_02_r_capsule
   center=<5.9786 -0.3059 6.113>
   rotation=<rot -11.3215 -90.7796 -93.3388>
  }]
 sleeve_03_l [{
   shapeType=Capsule
   length=16.965
   radius=10.7415
   shapeName=sleeve_03_r_capsule
   center=<12.1186 3.8747 7.1347>
   rotation=<rot 15.6558 -91.0645 -93.1039>
  }]
 sleeve_03_r [{
   shapeType=Capsule
   length=16.965
   radius=10.7415
   shapeName=sleeve_03_r_capsule
   center=<12.2258 -2.9743 4.5625>
   rotation=<rot -9.2796 -89.7373 -93.0283>
  }]
 sleeve_04_l [{
   shapeType=Capsule
   length=18.9397
   radius=8.7668
   shapeName=sleeve_04_r_capsule
   center=<11.299 2.0263 6.6906>
   rotation=<rot -5.8668 -89.4784 -91.966>
  }]
 sleeve_04_r [{
   shapeType=Capsule
   length=18.9397
   radius=8.7668
   shapeName=sleeve_04_r_capsule
   center=<11.0312 -1.813 6.5582>
   rotation=<rot -5.9998 -93.8843 -91.5104>
  }]
 sleeve_collision_01_l [{
   shapeType=Capsule
   length=22.6932
   radius=13.2683
   shapeName=sleeve_collision_01_l_capsule
   center=<17.587 -4.627 -12.7909>
   rotation=<rot 73.3198 90.008 179.999>
  }]
 sleeve_collision_01_r [{
   shapeType=Capsule
   length=22.6932
   radius=13.2683
   shapeName=sleeve_collision_01_r_capsule
   center=<17.5855 4.4803 -13.3903>
   rotation=<rot 73.0296 -89.9901 0.001>
  }]
 sleeve_collision_02_l [{
   shapeType=Capsule
   length=25.6705
   radius=14.6456
   shapeName=sleeve_collision_02_r_capsule
   center=<17.5867 -3.0707 -13.364>
   rotation=<rot 79.4326 90.0067 179.9977>
  }]
 sleeve_collision_02_r [{
   shapeType=Capsule
   length=25.6705
   radius=14.6456
   shapeName=sleeve_collision_02_r_capsule
   center=<17.5855 4.4018 -13.2859>
   rotation=<rot 76.4288 -89.9898 0.0013>
  }]
}
constraints [
 {
  bone1Name=sleeve_02_r
  bone2Name=sleeve_01_r
  swing1LimitDegrees=45
  swing1Motion=Locked
  swing2LimitDegrees=35
  swing2Motion=Locked
  twistLimitDegrees=10
  twistMotion=Locked
 }
 {
  bone1Name=sleeve_03_r
  bone2Name=sleeve_02_r
  swing1LimitDegrees=25
  swing2LimitDegrees=45
  swing2Motion=Locked
  twistLimitDegrees=15
  twistMotion=Locked
 }
 {bone1Name=sleeve_04_r, bone2Name=sleeve_03_r, swing1LimitDegrees=25, swing2LimitDegrees=45, twistLimitDegrees=15}
 {bone1Name=sleeve_04_l, bone2Name=sleeve_03_l, swing1LimitDegrees=25, swing2LimitDegrees=45, twistLimitDegrees=15}
 {
  bone1Name=sleeve_03_l
  bone2Name=sleeve_02_l
  swing1LimitDegrees=25
  swing2LimitDegrees=45
  swing2Motion=Locked
  twistLimitDegrees=15
  twistMotion=Locked
 }
 {
  bone1Name=sleeve_02_l
  bone2Name=sleeve_01_l
  swing1LimitDegrees=45
  swing1Motion=Locked
  swing2LimitDegrees=35
  swing2Motion=Locked
  twistLimitDegrees=10
  twistMotion=Locked
 }
]
