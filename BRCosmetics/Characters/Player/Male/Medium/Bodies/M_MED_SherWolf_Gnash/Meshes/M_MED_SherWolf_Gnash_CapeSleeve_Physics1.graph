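# Physics asset for the SherWolf Gnash male body's cape sleeves, defining capsule bodies and joint limits for both arms.
# Each side chains sleeve_01 through sleeve_04 with limited swing and twist, starting from a tiny capsule at sleeve_01.
# Forearm capsules and two large sleeve_collision capsules per side have no constraints, apparently serving as colliders for the sleeve chains.
# Several left-side shapes carry right-side capsule names.
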PhysicsAsset /BRCosmetics/Characters/Player/Male/Medium/Bodies/M_MED_SherWolf_Gnash/Meshes/M_MED_SherWolf_Gnash_CapeSleeve_Physics1
bodies {
 lowerarm_l [{shapeType=Capsule, length=19.6449, radius=6.8146, shapeName=lowerarm_l_capsule, center=<8.7884 0 0.053>, rotation=<rot 0 -90 -90.3456>}]
 lowerarm_r [{
   shapeType=Capsule
   length=18.7576
   radius=7.4342
   shapeName=lowerarm_r_capsule
   center=<-13.3632 0 -0.0806>
   rotation=<rot 0 -89.9999 -90.3456>
  }]
 sleeve_01_l [{shapeType=Capsule, length=0.505, radius=0.505, shapeName=sleeve_01_l_capsule, rotation=<rot 0 -90 0>}]
 sleeve_01_r [{shapeType=Capsule, length=0.505, radius=0.505, shapeName=sleeve_01_r_capsule, rotation=<rot 0 -90 0>}]
 sleeve_02_l [{
   shapeType=Capsule
   length=18.2984
   radius=9.4081
   shapeName=sleeve_02_r_capsule
   center=<10.8215 1.0806 7.5982>
   rotation=<rot 6.3022 -93.0058 -93.6798>
  }]
 sleeve_02_r [{
   shapeType=Capsule
   length=18.9397
   radius=8.7668
   shapeName=sleeve_02_r_capsule
   center=<10.7117 -1.5358 7.4607>
   rotation=<rot -11.3215 -90.7796 -93.3388>
  }]
 sleeve_03_l [{
   shapeType=Capsule
   length=18.9397
   radius=8.7668
   shapeName=sleeve_03_r_capsule
   center=<12.2881 1.5707 7.3033>
   rotation=<rot 13.664 -90.8764 -89.4416>
  }]
 sleeve_03_r [{
   shapeType=Capsule
   length=18.9397
   radius=8.7668
   shapeName=sleeve_03_r_capsule
   center=<12.2357 -1.0025 7.1586>
   rotation=<rot -9.2796 -89.7373 -88.8684>
  }]
 sleeve_04_l [{
   shapeType=Capsule
   length=18.9397
   radius=8.7668
   shapeName=sleeve_04_r_capsule
   center=<11.059 -1.6382 5.9704>
   rotation=<rot -5.9998 -93.8843 -91.5104>
  }]
 sleeve_04_r [{
   shapeType=Capsule
   length=18.9397
   radius=8.7668
   shapeName=sleeve_04_r_capsule
   center=<11.0236 -1.5534 5.1876>
   rotation=<rot -5.9998 -93.8843 -87.0118>
  }]
 sleeve_collision_01_l [{
   shapeType=Capsule
   length=37.3308
   radius=13.2683
   shapeName=sleeve_collision_01_l_capsule
   center=<17.587 -4.873 -14.5769>
   rotation=<rot 82.1589 90.0066 179.9975>
  }]
 sleeve_collision_01_r [{
   shapeType=Capsule
   length=37.3308
   radius=13.2683
   shapeName=sleeve_collision_01_r_capsule
   center=<17.5855 4.7191 -15.1772>
   rotation=<rot 82.3884 -89.9885 0.0026>
  }]
 sleeve_collision_02_l [{
   shapeType=Capsule
   length=40.3081
   radius=14.6456
   shapeName=sleeve_collision_02_r_capsule
   center=<17.5867 -2.806 -13.4557>
   rotation=<rot 88.2717 89.9933 179.9843>
  }]
 sleeve_collision_02_r [{
   shapeType=Capsule
   length=40.3081
   radius=14.6456
   shapeName=sleeve_collision_02_r_capsule
   center=<17.5855 4.4589 -14.7721>
   rotation=<rot 85.7876 -89.9863 0.0047>
  }]
}
constraints [
 {bone1Name=sleeve_02_r, bone2Name=sleeve_01_r, swing1LimitDegrees=15, swing2LimitDegrees=35, twistLimitDegrees=10}
 {bone1Name=sleeve_03_r, bone2Name=sleeve_02_r, swing1LimitDegrees=25, swing2LimitDegrees=45, twistLimitDegrees=7}
 {bone1Name=sleeve_04_r, bone2Name=sleeve_03_r, swing1LimitDegrees=20, swing2LimitDegrees=45, twistLimitDegrees=15}
 {bone1Name=sleeve_04_l, bone2Name=sleeve_03_l, swing1LimitDegrees=20, swing2LimitDegrees=45, twistLimitDegrees=15}
 {bone1Name=sleeve_03_l, bone2Name=sleeve_02_l, swing1LimitDegrees=25, swing2LimitDegrees=45, twistLimitDegrees=7}
 {bone1Name=sleeve_02_l, bone2Name=sleeve_01_l, swing1LimitDegrees=15, swing2LimitDegrees=35, twistLimitDegrees=10}
]
