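# Animation Blueprint for the SherWolf Gnash player body, adding secondary motion to its cape, sleeves, chain and necklace.
# The pose copied from the attached parent mesh passes through bone offsets, constraints, a Control Rig and Anim Dynamics.
# Cached poses then feed RigidBody cape simulations and trail controllers, switched by backbling, stance, emote, front-end, downed and cloth flags.
# A final RigidBody sleeve pass is layered per bone with the emote slot.
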AnimBlueprint /BRCosmetics/Characters/Player/Male/Medium/Bodies/M_MED_SherWolf_Gnash/Meshes/M_MED_SherWolf_Gnash_AnimBP
target_skeleton=&/M_MED_SherWolf_Gnash_Skeleton
graphs [
 AnimGraph {nodes [
   n1 = RigidBody!(
    ComponentPose=Local To Component(LocalPose=Use cached pose 'PreSleeveSim'())
    Alpha
    OverrideWorldGravity:Vector=Get Gravity Override Cape() -> (Gravity Override Cape:Vector)
    ExternalForce
    SimSpaceSettings:Sim Space Settings Structure=(WorldAlpha=0,VelocityScaleZ=0,DampingAlpha=1,MaxLinearVelocity=150,MaxAngularVelocity=1,MaxLinearAcceleration=300,MaxAngularAcceleration=50,ExternalLinearDragV=(X=0.2,Y=0.2,Z=0),ExternalLinearVelocity=(X=0,Y=0,Z=0),ExternalAngularVelocity=(X=0,Y=0,Z=0))
   ) -> (Pose) {
    bClampLinearTranslationLimitToRefPose=true
    bOverrideWorldGravity=true
    overridePhysicsAsset=&/M_MED_SherWolf_Gnash_CapeSleeve_Physics1
    simulationSpace=BaseBoneSpace
    alphaBoolBlend {blendOption=Linear}
    baseBoneRef {boneName=gravity}
    overrideWorldGravity {z=-980}
    simSpaceSettings {
     maxAngularAcceleration=50
     maxAngularVelocity=1
     maxLinearAcceleration=300
     maxLinearVelocity=150
     velocityScaleZ=0
     externalLinearDragV=<0.2 0.2>
    }
   }
   n2 = RigidBody(
    ComponentPose=@n1
    OverrideWorldGravity=Get Gravity Override Cape()
    SimSpaceSettings=(WorldAlpha=0,VelocityScaleZ=0,DampingAlpha=1,MaxLinearVelocity=150,MaxAngularVelocity=1,MaxLinearAcceleration=300,MaxAngularAcceleration=50,ExternalLinearDragV=(X=0.2,Y=0.2,Z=0),ExternalLinearVelocity=(X=0,Y=0,Z=0),ExternalAngularVelocity=(X=0,Y=0,Z=0))
   ) {
    bClampLinearTranslationLimitToRefPose=true
    bOverrideWorldGravity=true
    overridePhysicsAsset=&/M_MED_SherWolf_GnashBackBling_CapeSleeve_Physics2
    simulationSpace=BaseBoneSpace
    alphaBoolBlend {blendOption=Linear}
    baseBoneRef {boneName=gravity}
    overrideWorldGravity {z=-980}
    simSpaceSettings {
     maxAngularAcceleration=50
     maxAngularVelocity=1
     maxLinearAcceleration=300
     maxLinearVelocity=150
     velocityScaleZ=0
     externalLinearDragV=<0.2 0.2>
    }
   }
   n3 = "Transform (Modify) Bone - Bone: sleeve_01_r"!(ComponentPose=@n2, bAlphaBoolEnabled:bool=True, Translation="0,0,0", Rotation="1.5,0,0", Scale="1,1,1") -> (Pose) {
    alphaInputType=Bool
    rotationMode=BMM_Additive
    rotationSpace=BCS_BoneSpace
    translationMode=BMM_Additive
    translationSpace=BCS_BoneSpace
    alphaBoolBlend {blendOption=Linear}
    boneToModify {boneName=sleeve_01_r}
    rotation {pitch=1.5}
   }
   n4 = "Transform (Modify) Bone - Bone: sleeve_02_r"!(ComponentPose=@n3, bAlphaBoolEnabled:bool=True, Translation="0,0,0", Rotation="0,0,30", Scale="1,1,1") -> (Pose) {
    alphaInputType=Bool
    rotationMode=BMM_Additive
    rotationSpace=BCS_BoneSpace
    translationMode=BMM_Additive
    translationSpace=BCS_BoneSpace
    alphaBoolBlend {blendOption=Linear}
    boneToModify {boneName=sleeve_02_r}
    rotation {roll=30}
   }
   n5 = "Transform (Modify) Bone - Bone: sleeve_02_l"!(ComponentPose=@n4, bAlphaBoolEnabled:bool=True, Translation="0,0,0", Rotation="0,0,-20", Scale="1,1,1") -> (Pose) {
    alphaInputType=Bool
    rotationMode=BMM_Additive
    rotationSpace=BCS_BoneSpace
    translationMode=BMM_Additive
    translationSpace=BCS_BoneSpace
    alphaBoolBlend {blendOption=Linear}
    boneToModify {boneName=sleeve_02_l}
    rotation {roll=-20}
   }
   n6 = Layered blend per bone(
    BasePose=Blend Poses by bool(
     BlendPose_0=Use cached pose 'PreSleeveSim'()
     BlendPose_1=Component To Local(ComponentPose=@n5)
     BlendTime_1=0.7
     bActiveValue=Get emote_sherwolf() -> (emote_sherwolf:bool)
    ) {blendType=Linear}
    BlendPoses_0=Slot 'DefaultSlot'()
    BlendWeights_0=Get Is Playing Custom Emote() -> (Is Playing Custom Emote:double)
   ) {layerSetup [{branchFilters [
       {blendDepth=6, boneName=cape_mid_l_MSTR}
       {blendDepth=6, boneName=cape_mid_r_MSTR}
       {blendDepth=7, boneName=clavicle_cape_l}
       {blendDepth=7, boneName=clavicle_cape_r}
      ]}]}
   Output Pose(Result=@n6) {layerGroup=None}
   n7 = RigidBody!(
    ComponentPose=Local To Component(LocalPose=Use cached pose 'standingPoseCache'())
    Alpha
    OverrideWorldGravity=Get Gravity Override Cape()
    ExternalForce
    ComponentLinearAccScale:Vector="0,0,0"
    ComponentLinearVelScale:Vector="0,0,0"
    SimSpaceSettings=(WorldAlpha=1,VelocityScaleZ=1,DampingAlpha=1,MaxLinearVelocity=3000,MaxAngularVelocity=5,MaxLinearAcceleration=3000,MaxAngularAcceleration=3,ExternalLinearDragV=(X=0,Y=0,Z=0),ExternalLinearVelocity=(X=0,Y=0,Z=0),ExternalAngularVelocity=(X=0,Y=0,Z=0))
   ) -> (Pose) {
    bClampLinearTranslationLimitToRefPose=true
    bOverrideWorldGravity=true
    overridePhysicsAsset=&/M_MED_SherWolf_GnashBackBling_CapeFrontEnd_Physics1
    simulationSpace=BaseBoneSpace
    alphaBoolBlend {blendOption=Linear}
    baseBoneRef {boneName=root}
    overrideWorldGravity {z=-980}
    simSpaceSettings {maxAngularAcceleration=3, maxAngularVelocity=5, maxLinearAcceleration=3000, maxLinearVelocity=3000, worldAlpha=1}
   }
   n8 = RigidBody(
    ComponentPose=Local To Component(LocalPose=Use cached pose 'standingPoseCache'())
    OverrideWorldGravity=Get Gravity Override Cape()
    ComponentLinearAccScale="0,0,0"
    ComponentLinearVelScale="0,0,0"
    SimSpaceSettings=(WorldAlpha=1,VelocityScaleZ=1,DampingAlpha=1,MaxLinearVelocity=3000,MaxAngularVelocity=5,MaxLinearAcceleration=3000,MaxAngularAcceleration=3,ExternalLinearDragV=(X=0,Y=0,Z=0),ExternalLinearVelocity=(X=0,Y=0,Z=0),ExternalAngularVelocity=(X=0,Y=0,Z=0))
   ) {
    bClampLinearTranslationLimitToRefPose=true
    bOverrideWorldGravity=true
    overridePhysicsAsset=&/M_MED_SherWolf_Gnash_CapeFrontEnd_Physics
    simulationSpace=BaseBoneSpace
    alphaBoolBlend {blendOption=Linear}
    baseBoneRef {boneName=root}
    overrideWorldGravity {z=-980}
    simSpaceSettings {maxAngularAcceleration=3, maxAngularVelocity=5, maxLinearAcceleration=3000, maxLinearVelocity=3000, worldAlpha=1}
   }
   n9 = Blend Poses by bool(
    BlendPose_0=Blend Poses by bool(
     BlendPose_0=Component To Local(ComponentPose=@n7)
     BlendPose_1=Component To Local(ComponentPose=@n8)
     bActiveValue=Get Is Wearing Backbling()
    ) {blendType=Linear}
    BlendPose_1=Use cached pose 'posBackbing'()
    bActiveValue=Get FrontEnd or Emote() -> (FrontEnd or Emote:bool)
   ) {blendType=Linear}
   n10 = "Transform (Modify) Bone - Bone: gravity_cloth"!(
    ComponentPose=Local To Component(LocalPose=Use cached pose 'posBackbing'())
    bAlphaBoolEnabled:bool=True
    Translation="0,0,500"
    Rotation="0,0,0"
    Scale="1,1,1"
   ) -> (Pose) {
    alphaInputType=Bool
    rotationMode=BMM_Additive
    rotationSpace=BCS_BoneSpace
    translationMode=BMM_Additive
    translationSpace=BCS_BoneSpace
    alphaBoolBlend {blendOption=Linear}
    boneToModify {boneName=gravity_cloth}
    translation {z=500}
   }
   n11 = RigidBody!(
    ComponentPose=@n10
    Alpha=Get in vehicle rigid alpha() -> (in vehicle rigid alpha:double)
    OverrideWorldGravity=Get Gravity Override Cape()
    ExternalForce
    SimSpaceSettings=(WorldAlpha=0,VelocityScaleZ=0,DampingAlpha=1,MaxLinearVelocity=150,MaxAngularVelocity=1,MaxLinearAcceleration=300,MaxAngularAcceleration=50,ExternalLinearDragV=(X=0.2,Y=0.2,Z=0),ExternalLinearVelocity=(X=0,Y=0,Z=0),ExternalAngularVelocity=(X=0,Y=0,Z=0))
   ) -> (Pose) {
    bClampLinearTranslationLimitToRefPose=true
    bOverrideWorldGravity=true
    overridePhysicsAsset=&/M_MED_SherWolf_GnashBackBling_Cape_Physics2
    simulationSpace=BaseBoneSpace
    alphaBoolBlend {blendOption=Linear}
    baseBoneRef {boneName=gravity}
    overrideWorldGravity {z=-980}
    simSpaceSettings {
     maxAngularAcceleration=50
     maxAngularVelocity=1
     maxLinearAcceleration=300
     maxLinearVelocity=150
     velocityScaleZ=0
     externalLinearDragV=<0.2 0.2>
    }
   }
   n12 = "Transform (Modify) Bone - Bone: gravity_cloth"(
    ComponentPose=Local To Component(LocalPose=Use cached pose 'standingPoseCache'())
    bAlphaBoolEnabled=True
    Translation="0,0,500"
    Rotation="0,0,0"
    Scale="1,1,1"
   ) {
    alphaInputType=Bool
    rotationMode=BMM_Additive
    rotationSpace=BCS_BoneSpace
    translationMode=BMM_Additive
    translationSpace=BCS_BoneSpace
    alphaBoolBlend {blendOption=Linear}
    boneToModify {boneName=gravity_cloth}
    translation {z=500}
   }
   n13 = RigidBody!(
    ComponentPose=@n12
    Alpha
    OverrideWorldGravity=Get Gravity Override Cape()
    ExternalForce
    ComponentLinearAccScale:Vector="0,0,0"
    ComponentLinearVelScale:Vector="0,0,0"
    SimSpaceSettings=(WorldAlpha=1,VelocityScaleZ=1,DampingAlpha=1,MaxLinearVelocity=3000,MaxAngularVelocity=5,MaxLinearAcceleration=3000,MaxAngularAcceleration=3,ExternalLinearDragV=(X=0,Y=0,Z=0),ExternalLinearVelocity=(X=0,Y=0,Z=0),ExternalAngularVelocity=(X=0,Y=0,Z=0))
   ) -> (Pose) {
    bClampLinearTranslationLimitToRefPose=true
    bOverrideWorldGravity=true
    overridePhysicsAsset=&/M_MED_SherWolf_Gnash_Cape_Physics
    simulationSpace=BaseBoneSpace
    alphaBoolBlend {blendOption=Linear}
    baseBoneRef {boneName=pelvis}
    overrideWorldGravity {z=-980}
    simSpaceSettings {maxAngularAcceleration=3, maxAngularVelocity=5, maxLinearAcceleration=3000, maxLinearVelocity=3000, worldAlpha=1}
   }
   n14 = Blend Poses by bool(
    BlendPose_0=Component To Local(ComponentPose=@n11)
    BlendPose_1=Component To Local(ComponentPose=@n13)
    bActiveValue=Get Is Wearing Backbling()
   ) {blendType=Linear}
   n15 = Get Crouching or Emote or Siding() -> (Crouching or Emote or Siding:bool)
   Save cached pose 'postSim'(Pose=Blend Poses by bool(
     BlendPose_0=Use cached pose 'standingPoseCache'()
     BlendPose_1=Blend Poses by bool(BlendPose_0=@n14, BlendPose_1=@n9, BlendTime_1=0.7, bActiveValue=@n15) {blendType=Linear}
     BlendTime_1=0.7
     bActiveValue=Get emote_sherwolf()
    ) {blendType=Linear}) {cacheName=postSim}
   n16 = "Constraint - sleeve_01_l (PP)"!(
    ComponentPose=Local To Component(LocalPose=Use cached pose 'postSim'())
    bAlphaBoolEnabled:bool=True
    ConstraintWeights_0=0.5
    ConstraintWeights_1:float=0.5
   ) -> (Pose) {
    alphaInputType=Bool
    alphaBoolBlend {blendOption=Linear}
    boneToModify {boneName=sleeve_01_l}
    constraintSetup [{targetBone {boneName=cape_01_l}}, {targetBone {boneName=cape_mid_01_l}}]
    constraintWeights [1,1]
   }
   n17 = "Constraint - sleeve_01_r (PP)"!(ComponentPose=@n16, bAlphaBoolEnabled:bool=True, ConstraintWeights_0=0.5, ConstraintWeights_1:float=0.5) -> (Pose) {
    alphaInputType=Bool
    alphaBoolBlend {blendOption=Linear}
    boneToModify {boneName=sleeve_01_r}
    constraintSetup [{targetBone {boneName=cape_01_r}}, {targetBone {boneName=cape_mid_01_r}}]
    constraintWeights [1,1]
   }
   n18 = "Constraint - sleeve_02_l (PP)"!(ComponentPose=@n17, bAlphaBoolEnabled:bool=True, ConstraintWeights_0=0.5, ConstraintWeights_1:float=0.5) -> (Pose) {
    alphaInputType=Bool
    alphaBoolBlend {blendOption=Linear}
    boneToModify {boneName=sleeve_02_l}
    constraintSetup [{targetBone {boneName=cape_02_l}}, {targetBone {boneName=cape_mid_02_l}}]
    constraintWeights [1,1]
   }
   n19 = "Constraint - sleeve_02_r (PP)"!(ComponentPose=@n18, bAlphaBoolEnabled:bool=True, ConstraintWeights_0=0.5, ConstraintWeights_1:float=0.5) -> (Pose) {
    alphaInputType=Bool
    alphaBoolBlend {blendOption=Linear}
    boneToModify {boneName=sleeve_02_r}
    constraintSetup [{targetBone {boneName=cape_02_r}}, {targetBone {boneName=cape_mid_02_r}}]
    constraintWeights [1,1]
   }
   n20 = "Constraint - sleeve_collision_01_l (PP)"!(ComponentPose=@n19, bAlphaBoolEnabled:bool=True, ConstraintWeights_0=0.5, ConstraintWeights_1:float=0.5) -> (Pose) {
    alphaInputType=Bool
    alphaBoolBlend {blendOption=Linear}
    boneToModify {boneName=sleeve_collision_01_l}
    constraintSetup [{targetBone {boneName=cape_02_l}}, {targetBone {boneName=cape_mid_02_l}}]
    constraintWeights [1,1]
   }
   n21 = "Constraint - sleeve_collision_01_r (PP)"!(ComponentPose=@n20, bAlphaBoolEnabled:bool=True, ConstraintWeights_0=0.5, ConstraintWeights_1:float=0.5) -> (Pose) {
    alphaInputType=Bool
    alphaBoolBlend {blendOption=Linear}
    boneToModify {boneName=sleeve_collision_01_r}
    constraintSetup [{targetBone {boneName=cape_02_r}}, {targetBone {boneName=cape_mid_02_r}}]
    constraintWeights [1,1]
   }
   n22 = "Constraint - sleeve_collision_02_l (PP)"!(ComponentPose=@n21, bAlphaBoolEnabled:bool=True, ConstraintWeights_0=0.5, ConstraintWeights_1:float=0.5) -> (Pose) {
    alphaInputType=Bool
    alphaBoolBlend {blendOption=Linear}
    boneToModify {boneName=sleeve_collision_02_l}
    constraintSetup [{targetBone {boneName=cape_03_l}}, {targetBone {boneName=cape_mid_03_l}}]
    constraintWeights [1,1]
   }
   n23 = "Constraint - sleeve_collision_02_r (PP)"!(ComponentPose=@n22, bAlphaBoolEnabled:bool=True, ConstraintWeights_0=0.5, ConstraintWeights_1:float=0.5) -> (Pose) {
    alphaInputType=Bool
    alphaBoolBlend {blendOption=Linear}
    boneToModify {boneName=sleeve_collision_02_r}
    constraintSetup [{targetBone {boneName=cape_03_r}}, {targetBone {boneName=cape_mid_03_r}}]
    constraintWeights [1,1]
   }
   n24 = "Transform (Modify) Bone - Bone: cape_01_l"!(ComponentPose=@n23, bAlphaBoolEnabled:bool=Get bIsInFrontEnd(), Translation="-3,0,0", Rotation="0,0,0", Scale="1,1,1") -> (Pose) {
    alphaInputType=Bool
    rotationMode=BMM_Additive
    scaleMode=BMM_Additive
    translationMode=BMM_Additive
    alphaBoolBlend {blendOption=Linear}
    boneToModify {boneName=cape_01_l}
    translation {x=-3}
   }
   n25 = "Constraint - dyn_bag (PPP)"!(ComponentPose=@n24, Alpha=1, ConstraintWeights_0=0.8451, ConstraintWeights_1:float=0.0922, ConstraintWeights_2:float=0.0627) -> (Pose) {
    alphaBoolBlend {blendOption=Linear}
    boneToModify {boneName=dyn_bag}
    constraintSetup [{targetBone {boneName=pelvis}}, {targetBone {boneName=thigh_twist_01_l}}, {targetBone {boneName=deform_groin_l}}]
    constraintWeights [1,1,1]
   }
   Save cached pose 'RBANCache'(Pose=Component To Local(ComponentPose=@n25)) {cacheName=RBANCache}
   n26 = "Transform (Modify) Bone - Bone: attach_backpack"!(
    ComponentPose=Local To Component(LocalPose=Copy Pose From Mesh() {bUseAttachedParent=true})
    bAlphaBoolEnabled:bool=True
    Translation="0,1.5,0"
    Rotation="0,0,25"
    Scale="1,1,1"
   ) -> (Pose) {
    alphaInputType=Bool
    rotationMode=BMM_Additive
    rotationSpace=BCS_BoneSpace
    translationMode=BMM_Additive
    translationSpace=BCS_BoneSpace
    alphaBoolBlend {blendOption=Linear}
    boneToModify {boneName=attach_backpack}
    rotation {roll=25}
    translation {y=1.5}
   }
   n27 = "Transform (Modify) Bone - Bone: cape_mid_02_l"!(ComponentPose=@n26, bAlphaBoolEnabled:bool=True, Translation="0,0,0", Rotation="0,0,0", Scale="1,1,1") -> (Pose) {
    alphaInputType=Bool
    rotationMode=BMM_Additive
    rotationSpace=BCS_BoneSpace
    translationMode=BMM_Additive
    translationSpace=BCS_BoneSpace
    alphaBoolBlend {blendOption=Linear}
    boneToModify {boneName=cape_mid_02_l}
   }
   n28 = "Transform (Modify) Bone - Bone: cape_mid_02_r"!(ComponentPose=@n27, bAlphaBoolEnabled:bool=True, Translation="0,0,0", Rotation="0,0,0", Scale="1,1,1") -> (Pose) {
    alphaInputType=Bool
    rotationMode=BMM_Additive
    rotationSpace=BCS_BoneSpace
    translationMode=BMM_Additive
    translationSpace=BCS_BoneSpace
    alphaBoolBlend {blendOption=Linear}
    boneToModify {boneName=cape_mid_02_r}
   }
   n29 = "Constraint - dyn_Necklace_02 (PP)"!(ComponentPose=@n28, bAlphaBoolEnabled:bool=True, ConstraintWeights_0=0.3, ConstraintWeights_1:float=0.7) -> (Pose) {
    alphaInputType=Bool
    alphaBoolBlend {blendOption=Linear}
    boneToModify {boneName=dyn_Necklace_02}
    constraintSetup [{targetBone {boneName=spine_05}}, {targetBone {boneName=spine_04}}]
    constraintWeights [1,1]
   }
   n30 = Control Rig(Source=Component To Local(ComponentPose=@n29)) {
    alphaBoolBlend {blendOption=Linear}
    controlRigAssetReference {blueprintRigClass=&/M_MED_SherWolf_Gnash_CtrlRig.M_MED_SherWolf_Gnash_CtrlRig_C}
   }
   n31 = "Constraint - backpack_root (P)"!(ComponentPose=Local To Component(LocalPose=@n30), bAlphaBoolEnabled:bool=Get Is Wearing Backbling(), ConstraintWeights_0=1) -> (Pose) {
    alphaInputType=Bool
    alphaBoolBlend {blendOption=Linear}
    boneToModify {boneName=backpack_root}
    constraintSetup [{targetBone {boneName=spine_05}}]
    constraintWeights [1]
   }
   n32 = "Constraint - gravity (TT)"!(ComponentPose=@n31, bAlphaBoolEnabled:bool=True, ConstraintWeights_0=0.2, ConstraintWeights_1:float=0.8) -> (Pose) {
    alphaInputType=Bool
    alphaBoolBlend {blendOption=Linear}
    boneToModify {boneName=gravity}
    constraintSetup [{transformType=Translation, targetBone {boneName=root}}, {transformType=Translation, targetBone {boneName=spine_05}}]
    constraintWeights [1,1]
   }
   n33 = "Constraint - arms_l_fix (PP)"!(ComponentPose=@n32, bAlphaBoolEnabled:bool=True, ConstraintWeights_0=1, ConstraintWeights_1:float=0) -> (Pose) {
    alphaInputType=Bool
    alphaBoolBlend {blendOption=Linear}
    boneToModify {boneName=arms_l_fix}
    constraintSetup [{offsetOption=None, targetBone {boneName=root}}, {offsetOption=None, targetBone {boneName=spine_05}}]
    constraintWeights [1,1]
   }
   n34 = "Constraint - arms_r_fix (PP)"!(ComponentPose=@n33, bAlphaBoolEnabled:bool=True, ConstraintWeights_0=1, ConstraintWeights_1:float=0) -> (Pose) {
    alphaInputType=Bool
    alphaBoolBlend {blendOption=Linear}
    boneToModify {boneName=arms_r_fix}
    constraintSetup [{offsetOption=None, targetBone {boneName=root}}, {offsetOption=None, targetBone {boneName=spine_05}}]
    constraintWeights [1,1]
   }
   n35 = "Transform (Modify) Bone - Bone: cape_mid_02_r"(ComponentPose=@n34, bAlphaBoolEnabled=True, Translation="0,0,0", Rotation="-12,0,0", Scale="1,1,1") {
    alphaInputType=Bool
    rotationMode=BMM_Additive
    rotationSpace=BCS_BoneSpace
    translationMode=BMM_Additive
    translationSpace=BCS_BoneSpace
    alphaBoolBlend {blendOption=Linear}
    boneToModify {boneName=cape_mid_02_r}
    rotation {pitch=-12}
   }
   n36 = "Transform (Modify) Bone - Bone: cape_02_l"!(ComponentPose=@n35, bAlphaBoolEnabled:bool=True, Translation="0,0,0", Rotation="-20,0,0", Scale="1,1,1") -> (Pose) {
    alphaInputType=Bool
    rotationMode=BMM_Additive
    rotationSpace=BCS_BoneSpace
    translationMode=BMM_Additive
    translationSpace=BCS_BoneSpace
    alphaBoolBlend {blendOption=Linear}
    boneToModify {boneName=cape_02_l}
    rotation {pitch=-20}
   }
   n37 = "Transform (Modify) Bone - Bone: cape_mid_02_l"(ComponentPose=@n36, bAlphaBoolEnabled=True, Translation="0,0,0", Rotation="-12,0,0", Scale="1,1,1") {
    alphaInputType=Bool
    rotationMode=BMM_Additive
    rotationSpace=BCS_BoneSpace
    translationMode=BMM_Additive
    translationSpace=BCS_BoneSpace
    alphaBoolBlend {blendOption=Linear}
    boneToModify {boneName=cape_mid_02_l}
    rotation {pitch=-12}
   }
   n38 = "Transform (Modify) Bone - Bone: cape_02_r"!(ComponentPose=@n37, bAlphaBoolEnabled:bool=True, Translation="0,0,0", Rotation="-20,0,0", Scale="1,1,1") -> (Pose) {
    alphaInputType=Bool
    rotationMode=BMM_Additive
    rotationSpace=BCS_BoneSpace
    translationMode=BMM_Additive
    translationSpace=BCS_BoneSpace
    alphaBoolBlend {blendOption=Linear}
    boneToModify {boneName=cape_02_r}
    rotation {pitch=-20}
   }
   n39 = "Transform (Modify) Bone - Bone: cape_mid_04_r"!(ComponentPose=@n38, bAlphaBoolEnabled:bool=True, Translation="0,0,0", Rotation="-20,0,0", Scale="1,1,1") -> (Pose) {
    alphaInputType=Bool
    rotationMode=BMM_Additive
    rotationSpace=BCS_BoneSpace
    translationMode=BMM_Additive
    translationSpace=BCS_BoneSpace
    alphaBoolBlend {blendOption=Linear}
    boneToModify {boneName=cape_mid_04_r}
    rotation {pitch=-20}
   }
   n40 = "Transform (Modify) Bone - Bone: cape_04_l"!(ComponentPose=@n39, bAlphaBoolEnabled:bool=True, Translation="0,0,0", Rotation="-20,0,0", Scale="1,1,1") -> (Pose) {
    alphaInputType=Bool
    rotationMode=BMM_Additive
    rotationSpace=BCS_BoneSpace
    translationMode=BMM_Additive
    translationSpace=BCS_BoneSpace
    alphaBoolBlend {blendOption=Linear}
    boneToModify {boneName=cape_04_l}
    rotation {pitch=-20}
   }
   n41 = "Transform (Modify) Bone - Bone: cape_mid_04_l"!(ComponentPose=@n40, bAlphaBoolEnabled:bool=True, Translation="0,0,0", Rotation="-20,0,0", Scale="1,1,1") -> (Pose) {
    alphaInputType=Bool
    rotationMode=BMM_Additive
    rotationSpace=BCS_BoneSpace
    translationMode=BMM_Additive
    translationSpace=BCS_BoneSpace
    alphaBoolBlend {blendOption=Linear}
    boneToModify {boneName=cape_mid_04_l}
    rotation {pitch=-20}
   }
   n42 = "Transform (Modify) Bone - Bone: cape_04_r"!(ComponentPose=@n41, bAlphaBoolEnabled:bool=True, Translation="0,0,0", Rotation="-20,0,0", Scale="1,1,1") -> (Pose) {
    alphaInputType=Bool
    rotationMode=BMM_Additive
    rotationSpace=BCS_BoneSpace
    translationMode=BMM_Additive
    translationSpace=BCS_BoneSpace
    alphaBoolBlend {blendOption=Linear}
    boneToModify {boneName=cape_04_r}
    rotation {pitch=-20}
   }
   n43 = "Transform (Modify) Bone - Bone: gravity_cloth"(ComponentPose=@n42, bAlphaBoolEnabled=True, Translation="0,0,-100", Rotation="0,0,0", Scale="1,1,1") {
    alphaInputType=Bool
    rotationMode=BMM_Additive
    rotationSpace=BCS_BoneSpace
    translationMode=BMM_Additive
    translationSpace=BCS_BoneSpace
    alphaBoolBlend {blendOption=Linear}
    boneToModify {boneName=gravity_cloth}
    translation {z=-100}
   }
   n44 = "Transform (Modify) Bone - Bone: backpack_collision"!(ComponentPose=@n43, bAlphaBoolEnabled:bool=True, Translation="0,0,300", Rotation="0,0,0", Scale="1,1,1") -> (Pose) {
    alphaInputType=Bool
    rotationMode=BMM_Additive
    rotationSpace=BCS_BoneSpace
    translationMode=BMM_Additive
    translationSpace=BCS_BoneSpace
    alphaBoolBlend {blendOption=Linear}
    boneToModify {boneName=backpack_collision}
    translation {z=300}
   }
   n45 = Anim Dynamics - Bone: dyn_Chain!(ComponentPose=@n44, Alpha=1, GravityOverride:Vector="0,0,500", ExternalForce="0,0,0") -> (Pose) {
    angularBiasOverride=1
    angularDampingOverride=0.99
    bOverrideAngularBias=true
    bOverrideAngularDamping=true
    bUseGravityOverride=true
    gravityScale=0.3
    simulationSpace=BoneRelative
    alphaBoolBlend {blendOption=Linear}
    boundBone {boneName=dyn_Chain}
    componentAppliedLinearAccClamp=<2000 2000 1600>
    componentLinearAccScale=<0.4 0.4 0.4>
    componentLinearVelScale=<0.45 0.45 0.55>
    gravityOverride {z=500}
    physicsBodyDefinitions [{
      sphereCollisionRadius=2
      boundBone {boneName=dyn_Chain}
      boxExtents=<8 15 15>
      constraintSetup {twistAxis=AxisY, angularLimitsMax=<0 -2 10>, angularLimitsMin=<0 -20 -10>, angularTarget=<0 0 0>}
      localJointOffset=<-4 -1 1>
     }]
    relativeSpaceBone {boneName=root}
   }
   n46 = Anim Dynamics - Bone: dyn_Chain(ComponentPose=@n45, Alpha=1, GravityOverride="0,0,500", ExternalForce="0,0,0") {
    angularBiasOverride=1
    angularDampingOverride=0.99
    bOverrideAngularBias=true
    bOverrideAngularDamping=true
    bUseGravityOverride=true
    gravityScale=0.3
    simulationSpace=BoneRelative
    alphaBoolBlend {blendOption=Linear}
    boundBone {boneName=dyn_Chain}
    componentAppliedLinearAccClamp=<2000 2000 1600>
    componentLinearAccScale=<0.4 0.4 0.4>
    componentLinearVelScale=<0.45 0.45 0.55>
    gravityOverride {z=500}
    physicsBodyDefinitions [{
      sphereCollisionRadius=2
      boundBone {boneName=dyn_Chain}
      boxExtents=<20 20 20>
      constraintSetup {twistAxis=AxisY, angularLimitsMax=<0 -2 5>, angularLimitsMin=<0 -20 -5>, angularTarget=<0 0 0>}
      localJointOffset=<-4 -1 1>
     }]
    relativeSpaceBone {boneName=root}
   }
   n47 = Anim Dynamics - Bone: dyn_Necklace!(ComponentPose=@n46, Alpha=1, GravityOverride:Vector="0,0,500", ExternalForce="0,0,0") -> (Pose) {
    angularBiasOverride=1
    angularDampingOverride=0.999
    bOverrideAngularBias=true
    bOverrideAngularDamping=true
    bUseGravityOverride=true
    gravityScale=0.3
    simulationSpace=BoneRelative
    alphaBoolBlend {blendOption=Linear}
    boundBone {boneName=dyn_Necklace}
    componentAppliedLinearAccClamp=<2000 2000 1600>
    componentLinearAccScale=<0.4 0.4 0.4>
    componentLinearVelScale=<0.45 0.45 0.55>
    gravityOverride {z=500}
    physicsBodyDefinitions [{
      sphereCollisionRadius=2
      boundBone {boneName=dyn_Necklace}
      boxExtents=<8 15 15>
      constraintSetup {twistAxis=AxisZ, angularLimitsMax=<0 2 -5>, angularLimitsMin=<0 -2 -20>, angularTarget=<0 0 0>}
      localJointOffset=<-4 -0.1 0>
     }]
    relativeSpaceBone {boneName=root}
   }
   n48 = Anim Dynamics - Bone: dyn_Necklace_00!(ComponentPose=@n47, Alpha=1, GravityOverride:Vector="0,0,500", ExternalForce="0,0,0") -> (Pose) {
    angularBiasOverride=1
    angularDampingOverride=0.999
    bOverrideAngularBias=true
    bOverrideAngularDamping=true
    bUseGravityOverride=true
    gravityScale=0.3
    simulationSpace=BoneRelative
    alphaBoolBlend {blendOption=Linear}
    boundBone {boneName=dyn_Necklace_00}
    componentAppliedLinearAccClamp=<2000 2000 1600>
    componentLinearAccScale=<0.4 0.4 0.4>
    componentLinearVelScale=<0.45 0.45 0.55>
    gravityOverride {z=500}
    physicsBodyDefinitions [{
      sphereCollisionRadius=2
      boundBone {boneName=dyn_Necklace_00}
      boxExtents=<8 15 15>
      constraintSetup {twistAxis=AxisZ, angularLimitsMax=<0 5 -5>, angularLimitsMin=<0 -5 -20>, angularTarget=<0 0 0>}
      localJointOffset=<-4 -0.1 0>
     }]
    relativeSpaceBone {boneName=root}
   }
   n49 = Anim Dynamics - Bone: dyn_Necklace_00(ComponentPose=@n48, Alpha=1, GravityOverride="0,0,500", ExternalForce="0,0,0") {
    angularBiasOverride=1
    angularDampingOverride=0.999
    bOverrideAngularBias=true
    bOverrideAngularDamping=true
    bUseGravityOverride=true
    gravityScale=0.3
    simulationSpace=BoneRelative
    alphaBoolBlend {blendOption=Linear}
    boundBone {boneName=dyn_Necklace_00}
    componentAppliedLinearAccClamp=<2000 2000 1600>
    componentLinearAccScale=<0.4 0.4 0.4>
    componentLinearVelScale=<0.45 0.45 0.55>
    gravityOverride {z=500}
    physicsBodyDefinitions [{
      sphereCollisionRadius=2
      boundBone {boneName=dyn_Necklace_00}
      boxExtents=<13 13 13>
      constraintSetup {twistAxis=AxisY, angularLimitsMax=<0 2 0>, angularLimitsMin=<0 -2 -5>, angularTarget=<0 0 0>}
      localJointOffset=<-4 -0.1 0>
     }]
    relativeSpaceBone {boneName=root}
   }
   n50 = Anim Dynamics - Bone: dyn_Necklace_01!(ComponentPose=@n49, Alpha=1, GravityOverride:Vector="0,0,500", ExternalForce="0,0,0") -> (Pose) {
    angularBiasOverride=1
    angularDampingOverride=0.999
    bOverrideAngularBias=true
    bOverrideAngularDamping=true
    bUseGravityOverride=true
    gravityScale=0.3
    simulationSpace=BoneRelative
    alphaBoolBlend {blendOption=Linear}
    boundBone {boneName=dyn_Necklace_01}
    componentAppliedLinearAccClamp=<2000 2000 1600>
    componentLinearAccScale=<0.4 0.4 0.4>
    componentLinearVelScale=<0.45 0.45 0.55>
    gravityOverride {z=500}
    physicsBodyDefinitions [{
      sphereCollisionRadius=2
      boundBone {boneName=dyn_Necklace_01}
      boxExtents=<5 5 5>
      constraintSetup {angularLimitsMax=<0 2 2>, angularLimitsMin=<0 -2 -30>, angularTarget=<0 0 0>}
      localJointOffset=<-2 -0.2 0>
     }]
    relativeSpaceBone {boneName=root}
   }
   n51 = Anim Dynamics - Bone: dyn_Necklace_02!(ComponentPose=@n50, Alpha=1, GravityOverride:Vector="0,0,500", ExternalForce="0,0,0") -> (Pose) {
    angularBiasOverride=1
    angularDampingOverride=0.999
    bOverrideAngularBias=true
    bOverrideAngularDamping=true
    bUseGravityOverride=true
    gravityScale=0.3
    simulationSpace=BoneRelative
    alphaBoolBlend {blendOption=Linear}
    boundBone {boneName=dyn_Necklace_02}
    componentAppliedLinearAccClamp=<2000 2000 1600>
    componentLinearAccScale=<0.4 0.4 0.4>
    componentLinearVelScale=<0.45 0.45 0.55>
    gravityOverride {z=500}
    physicsBodyDefinitions [{
      sphereCollisionRadius=2
      boundBone {boneName=dyn_Necklace_02}
      boxExtents=<8 15 15>
      constraintSetup {twistAxis=AxisZ, angularLimitsMax=<0 1 -17>, angularLimitsMin=<0 -1 -25>, angularTarget=<0 0 0>}
      localJointOffset=<-4 -0.1 0>
     }]
    relativeSpaceBone {boneName=root}
   }
   n52 = Anim Dynamics - Bone: dyn_Necklace_02(ComponentPose=@n51, Alpha=1, GravityOverride="0,0,500", ExternalForce="0,0,0") {
    angularBiasOverride=1
    angularDampingOverride=0.999
    bOverrideAngularBias=true
    bOverrideAngularDamping=true
    bUseGravityOverride=true
    gravityScale=0.3
    simulationSpace=BoneRelative
    alphaBoolBlend {blendOption=Linear}
    boundBone {boneName=dyn_Necklace_02}
    componentAppliedLinearAccClamp=<2000 2000 1600>
    componentLinearAccScale=<0.4 0.4 0.4>
    componentLinearVelScale=<0.45 0.45 0.55>
    gravityOverride {z=500}
    physicsBodyDefinitions [{
      sphereCollisionRadius=2
      boundBone {boneName=dyn_Necklace_02}
      boxExtents=<10 10 10>
      constraintSetup {twistAxis=AxisZ, angularLimitsMax=<0 5 -10>, angularLimitsMin=<0 -5 -20>, angularTarget=<0 0 0>}
      localJointOffset=<-4 -0.1 0>
     }]
    relativeSpaceBone {boneName=root}
   }
   n53 = "Transform (Modify) Bone - Bone: gravity_cloth"(ComponentPose=@n52, bAlphaBoolEnabled=True, Translation="0,0,-500", Rotation="0,0,0", Scale="1,1,1") {
    alphaInputType=Bool
    rotationMode=BMM_Additive
    rotationSpace=BCS_BoneSpace
    translationMode=BMM_Additive
    translationSpace=BCS_BoneSpace
    alphaBoolBlend {blendOption=Linear}
    boneToModify {boneName=gravity_cloth}
    translation {z=-500}
   }
   Save cached pose 'standingPoseCache'(Pose=Component To Local(ComponentPose=@n53)) {cacheName=standingPoseCache}
   n54 = "Transform (Modify) Bone - Bone: backpack_collision"(
    ComponentPose=Local To Component(LocalPose=Use cached pose 'standingPoseCache'())
    bAlphaBoolEnabled=True
    Translation="0,0,300"
    Rotation="0,0,0"
    Scale="1,1,1"
   ) {
    alphaInputType=Bool
    rotationMode=BMM_Additive
    rotationSpace=BCS_BoneSpace
    translationMode=BMM_Additive
    translationSpace=BCS_BoneSpace
    alphaBoolBlend {blendOption=Linear}
    boneToModify {boneName=backpack_collision}
    translation {z=300}
   }
   n55 = Get BwdMovement() -> (BwdMovement:double)
   n56 = Trail controller - Bone: cape_mid_06_l(ComponentPose=Local To Component(LocalPose=Use cached pose 'RBANCache'()), Alpha=@n55) {
    bUsePlanarLimit=true
    chainLength=5
    debugLifeTime=0.1
    maxDeltaTime=0.0333
    alphaScaleBiasClamp {bInterpResult=true}
    rotationLimits [
     {limitMax=<180 180 180>}
     {limitMax=<180 180 180>}
     {limitMax=<180 90 180>, limitMin=<-180 -90 -180>}
     {limitMax=<180 180 180>}
     {limitMax=<180 180 180>}
    ]
    rotationOffsets [<0 0 0>, <0 0 0>, <0 0 0>, <0 0 0>, <0 0 0>]
    trailBone {boneName=cape_mid_06_l}
    trailRelaxationSpeed {editorCurveData {keys [<at 0 15 RCIM_Linear>,<at 0.5289 13.1838 RCIM_Linear>,<at 1 6 RCIM_Linear>]}}
   }
   n57 = Trail controller - Bone: cape_mid_06_r(ComponentPose=@n56, Alpha=@n55) {
    bUsePlanarLimit=true
    chainLength=5
    debugLifeTime=0.1
    maxDeltaTime=0.0333
    alphaScaleBiasClamp {bInterpResult=true}
    rotationLimits [
     {limitMax=<180 180 180>}
     {limitMax=<180 180 180>}
     {limitMax=<180 90 180>, limitMin=<-180 -90 -180>}
     {limitMax=<180 180 180>}
     {limitMax=<180 180 180>}
    ]
    rotationOffsets [<0 0 0>, <0 0 0>, <0 0 0>, <0 0 0>, <0 0 0>]
    trailBone {boneName=cape_mid_06_r}
    trailRelaxationSpeed {editorCurveData {keys [<at 0 15 RCIM_Linear>,<at 0.5289 13.1838 RCIM_Linear>,<at 1 6 RCIM_Linear>]}}
   }
   n58 = Trail controller - Bone: cape_06_l(ComponentPose=@n57, Alpha=@n55) {
    bUsePlanarLimit=true
    chainLength=6
    debugLifeTime=0.1
    maxDeltaTime=0.0333
    alphaScaleBiasClamp {bInterpResult=true}
    rotationLimits [
     {limitMax=<180 180 180>}
     {limitMax=<180 180 180>}
     {limitMax=<180 90 180>, limitMin=<-180 -90 -180>}
     {limitMax=<180 180 180>}
     {limitMax=<180 180 180>}
     {limitMax=<180 180 180>}
    ]
    rotationOffsets [<0 0 0>, <0 0 0>, <0 0 0>, <0 0 0>, <0 0 0>, <0 0 0>]
    trailBone {boneName=cape_06_l}
    trailRelaxationSpeed {editorCurveData {keys [<at 0 15 RCIM_Linear>,<at 0.5289 13.1838 RCIM_Linear>,<at 1 6 RCIM_Linear>]}}
   }
   n59 = Trail controller - Bone: cape_06_r(ComponentPose=@n58, Alpha=@n55) {
    bUsePlanarLimit=true
    chainLength=6
    debugLifeTime=0.1
    maxDeltaTime=0.0333
    alphaScaleBiasClamp {bInterpResult=true}
    rotationLimits [
     {limitMax=<180 180 180>}
     {limitMax=<180 180 180>}
     {limitMax=<180 90 180>, limitMin=<-180 -90 -180>}
     {limitMax=<180 180 180>}
     {limitMax=<180 180 180>}
     {limitMax=<180 180 180>}
    ]
    rotationOffsets [<0 0 0>, <0 0 0>, <0 0 0>, <0 0 0>, <0 0 0>, <0 0 0>]
    trailBone {boneName=cape_06_r}
    trailRelaxationSpeed {editorCurveData {keys [<at 0 15 RCIM_Linear>,<at 0.5289 13.1838 RCIM_Linear>,<at 1 6 RCIM_Linear>]}}
   }
   n60 = "Constraint - sleeve_01_l (PP)"(ComponentPose=@n59, bAlphaBoolEnabled=True, ConstraintWeights_0=0.5, ConstraintWeights_1=0.5) {
    alphaInputType=Bool
    alphaBoolBlend {blendOption=Linear}
    boneToModify {boneName=sleeve_01_l}
    constraintSetup [{targetBone {boneName=cape_01_l}}, {targetBone {boneName=cape_mid_01_l}}]
    constraintWeights [1,1]
   }
   n61 = "Constraint - sleeve_01_r (PP)"(ComponentPose=@n60, bAlphaBoolEnabled=True, ConstraintWeights_0=0.5, ConstraintWeights_1=0.5) {
    alphaInputType=Bool
    alphaBoolBlend {blendOption=Linear}
    boneToModify {boneName=sleeve_01_r}
    constraintSetup [{targetBone {boneName=cape_01_r}}, {targetBone {boneName=cape_mid_01_r}}]
    constraintWeights [1,1]
   }
   n62 = "Constraint - sleeve_02_l (PP)"(ComponentPose=@n61, bAlphaBoolEnabled=True, ConstraintWeights_0=0.5, ConstraintWeights_1=0.5) {
    alphaInputType=Bool
    alphaBoolBlend {blendOption=Linear}
    boneToModify {boneName=sleeve_02_l}
    constraintSetup [{targetBone {boneName=cape_02_l}}, {targetBone {boneName=cape_mid_02_l}}]
    constraintWeights [1,1]
   }
   n63 = "Constraint - sleeve_02_r (PP)"(ComponentPose=@n62, bAlphaBoolEnabled=True, ConstraintWeights_0=0.5, ConstraintWeights_1=0.5) {
    alphaInputType=Bool
    alphaBoolBlend {blendOption=Linear}
    boneToModify {boneName=sleeve_02_r}
    constraintSetup [{targetBone {boneName=cape_02_r}}, {targetBone {boneName=cape_mid_02_r}}]
    constraintWeights [1,1]
   }
   n64 = Get BwdMovement()
   n65 = Trail controller - Bone: cape_mid_06_l(ComponentPose=Local To Component(LocalPose=Use cached pose 'RBANCache'()), Alpha=@n64) {
    bUsePlanarLimit=true
    chainLength=4
    debugLifeTime=0.1
    maxDeltaTime=0.0333
    alphaScaleBiasClamp {bInterpResult=true}
    rotationLimits [{limitMax=<180 180 180>}, {limitMax=<180 180 180>}, {limitMax=<180 90 180>, limitMin=<-180 -90 -180>}, {limitMax=<180 180 180>}]
    rotationOffsets [<0 0 0>, <0 0 0>, <0 0 0>, <0 0 0>]
    trailBone {boneName=cape_mid_06_l}
    trailRelaxationSpeed {editorCurveData {keys [<at 0 15 RCIM_Linear>,<at 0.5289 13.1838 RCIM_Linear>,<at 1 6 RCIM_Linear>]}}
   }
   n66 = Trail controller - Bone: cape_mid_06_r(ComponentPose=@n65, Alpha=@n64) {
    bUsePlanarLimit=true
    chainLength=4
    debugLifeTime=0.1
    maxDeltaTime=0.0333
    alphaScaleBiasClamp {bInterpResult=true}
    rotationLimits [{limitMax=<180 180 180>}, {limitMax=<180 180 180>}, {limitMax=<180 180 180>}, {limitMax=<180 180 180>}]
    rotationOffsets [<0 0 0>, <0 0 0>, <0 0 0>, <0 0 0>]
    trailBone {boneName=cape_mid_06_r}
    trailRelaxationSpeed {editorCurveData {keys [<at 0 15 RCIM_Linear>,<at 0.5289 13.1838 RCIM_Linear>,<at 1 6 RCIM_Linear>]}}
   }
   n67 = Trail controller - Bone: cape_06_l(ComponentPose=@n66, Alpha=@n64) {
    bUsePlanarLimit=true
    chainLength=4
    debugLifeTime=0.1
    maxDeltaTime=0.0333
    alphaScaleBiasClamp {bInterpResult=true}
    rotationLimits [{limitMax=<180 180 180>}, {limitMax=<180 180 180>}, {limitMax=<180 90 180>, limitMin=<-180 -90 -180>}, {limitMax=<180 180 180>}]
    rotationOffsets [<0 0 0>, <0 0 0>, <0 0 0>, <0 0 0>]
    trailBone {boneName=cape_06_l}
    trailRelaxationSpeed {editorCurveData {keys [<at 0 15 RCIM_Linear>,<at 0.5289 13.1838 RCIM_Linear>,<at 1 6 RCIM_Linear>]}}
   }
   n68 = Trail controller - Bone: cape_06_r(ComponentPose=@n67, Alpha=@n64) {
    bUsePlanarLimit=true
    chainLength=4
    debugLifeTime=0.1
    maxDeltaTime=0.0333
    alphaScaleBiasClamp {bInterpResult=true}
    rotationLimits [{limitMax=<180 180 180>}, {limitMax=<180 180 180>}, {limitMax=<180 90 180>, limitMin=<-180 -90 -180>}, {limitMax=<180 180 180>}]
    rotationOffsets [<0 0 0>, <0 0 0>, <0 0 0>, <0 0 0>]
    trailBone {boneName=cape_06_r}
    trailRelaxationSpeed {editorCurveData {keys [<at 0 15 RCIM_Linear>,<at 0.5289 13.1838 RCIM_Linear>,<at 1 6 RCIM_Linear>]}}
   }
   n69 = Blend Poses by bool(
    BlendPose_0=Component To Local(ComponentPose=@n68)
    BlendPose_1=Component To Local(ComponentPose=@n63)
    bActiveValue=Get Is Wearing Backbling()
   ) {blendType=Linear}
   n70 = Use cached pose 'RBANCache'()
   n71 = Get NotClothEnabled() -> (NotClothEnabled:bool)
   n72 = Get Is_Emoting_FE() -> (Is_Emoting_FE:bool)
   n73 = "Transform (Modify) Bone - Bone: cape_03_r"!(
    ComponentPose=Local To Component(LocalPose=Blend Poses by bool(BlendPose_0=@n70, BlendPose_1=@n69, bActiveValue=@n71) {blendType=Linear})
    bAlphaBoolEnabled:bool=@n72
    Translation="0,0,0"
    Rotation="15,0,0"
    Scale="1,1,1"
   ) -> (Pose) {
    alphaInputType=Bool
    rotationMode=BMM_Additive
    rotationSpace=BCS_BoneSpace
    alphaBoolBlend {blendInTime=1, blendOption=CubicInOut, blendOutTime=1}
    boneToModify {boneName=cape_03_r}
    rotation {pitch=15}
   }
   n74 = "Transform (Modify) Bone - Bone: cape_mid_03_r"!(ComponentPose=@n73, bAlphaBoolEnabled:bool=@n72, Translation="0,0,0", Rotation="15,0,0", Scale="1,1,1") -> (Pose) {
    alphaInputType=Bool
    rotationMode=BMM_Additive
    rotationSpace=BCS_BoneSpace
    alphaBoolBlend {blendInTime=1, blendOption=CubicInOut, blendOutTime=1}
    boneToModify {boneName=cape_mid_03_r}
    rotation {pitch=15}
   }
   Save cached pose 'PreSleeveSim'(Pose=Blend Poses by bool(BlendPose_0=Component To Local(ComponentPose=@n54), BlendPose_1=Component To Local(ComponentPose=@n74), bActiveValue=Get bIsDBNO()) {blendType=Linear}) {cacheName=PreSleeveSim}
   n75 = RigidBody(
    ComponentPose=Local To Component(LocalPose=Use cached pose 'standingPoseCache'())
    OverrideWorldGravity=Get Gravity Override Cape()
    ComponentLinearAccScale="0.5,0.3,0.3"
    ComponentLinearVelScale="0.5,0.3,0.3"
    SimSpaceSettings=(WorldAlpha=1,VelocityScaleZ=1,DampingAlpha=1,MaxLinearVelocity=3000,MaxAngularVelocity=5,MaxLinearAcceleration=3000,MaxAngularAcceleration=3,ExternalLinearDragV=(X=0,Y=0,Z=0),ExternalLinearVelocity=(X=0,Y=0,Z=0),ExternalAngularVelocity=(X=0,Y=0,Z=0))
   ) {
    bClampLinearTranslationLimitToRefPose=true
    bOverrideWorldGravity=true
    overridePhysicsAsset=&M_MED_SherWolf_Gnash_Cape_Physics
    simulationSpace=BaseBoneSpace
    alphaBoolBlend {blendOption=Linear}
    baseBoneRef {boneName=gravity}
    overrideWorldGravity {z=-980}
    simSpaceSettings {maxAngularAcceleration=3, maxAngularVelocity=5, maxLinearAcceleration=3000, maxLinearVelocity=3000, worldAlpha=1}
   }
   n76 = Component To Local(ComponentPose=@n75)
   n77 = RigidBody!(
    ComponentPose=Local To Component(LocalPose=Use cached pose 'standingPoseCache'())
    Alpha=Get in vehicle rigid alpha()
    OverrideWorldGravity=Get Gravity Override Cape()
    ExternalForce
   ) -> (Pose) {
    bClampLinearTranslationLimitToRefPose=true
    bOverrideWorldGravity=true
    overridePhysicsAsset=&M_MED_SherWolf_Gnash_Cape_Physics
    simulationSpace=BaseBoneSpace
    alphaBoolBlend {blendOption=Linear}
    baseBoneRef {boneName=gravity}
    componentAppliedLinearAccClamp=<3000 3000 2500>
    componentLinearAccScale=<0.05 0.05 0.25>
    componentLinearVelScale=<0.05 0.05 0.25>
    overrideWorldGravity {z=-980}
    simSpaceSettings {
     maxAngularAcceleration=50
     maxAngularVelocity=1
     maxLinearAcceleration=300
     maxLinearVelocity=150
     velocityScaleZ=0
     externalLinearDragV=<0.2 0.2>
    }
   }
   n78 = Component To Local(ComponentPose=@n77)
   n79 = Get bIsTacticalSprinting()
   n80 = RigidBody(
    ComponentPose=Local To Component(LocalPose=Use cached pose 'standingPoseCache'())
    Alpha=Get in vehicle rigid alpha()
    OverrideWorldGravity=Get Gravity Override Cape()
   ) {
    bClampLinearTranslationLimitToRefPose=true
    bOverrideWorldGravity=true
    overridePhysicsAsset=&M_MED_SherWolf_GnashBackBling_Cape_Physics2
    simulationSpace=BaseBoneSpace
    alphaBoolBlend {blendOption=Linear}
    baseBoneRef {boneName=gravity}
    componentLinearAccScale=<0.7 0.3 0.3>
    componentLinearVelScale=<0.7 0.3 0.3>
    overrideWorldGravity {z=-980}
    simSpaceSettings {
     maxAngularAcceleration=50
     maxAngularVelocity=1
     maxLinearAcceleration=300
     maxLinearVelocity=150
     worldAlpha=1
     externalLinearDragV=<0.2 0.2>
    }
   }
   Save cached pose 'posBackbing'(Pose=Blend Poses by bool(
     BlendPose_0=Component To Local(ComponentPose=@n80)
     BlendPose_1=Blend Poses by bool(BlendPose_0=@n78, BlendPose_1=@n76, BlendTime_1=0.7, bActiveValue=@n79) {blendType=Linear}
     BlendTime_1=0.7
     bActiveValue=Get Is Wearing Backbling()
    ) {blendType=Linear}) {cacheName=posBackbing}
  ]}
 EventGraph {nodes [
   n1 = Cloth Controller Function DEPRECATED!(
    execute
    CharacterPartAnimInstance:Custom Character Part Anim Instance Object Reference=Self-Reference()
    ClothingAssetName:String=M_MED_SherWolf_Gnash_Clothing_2
    Cloth Paramaters:Cloth Params Struct Structure=Get Cloth Paramaters_0() -> (Cloth Paramaters_0:Cloth Params Struct Structure)
    Cloth Core Settings:Cloth Core Setting Struct Structure=Get Cloth Core Settings_0() -> (Cloth Core Settings_0:Cloth Core Setting Struct Structure)
    ClothUpAxis:Vector="1,0,0"
    OnGround_Moving_SpeedLimit:double=2
    OnGround_Moving_AngularVel_Max:double=0.05
    OnGround_Moving_FictitiousVel_Max:double=0.6
    LinVelScale_FwdBwd_Clamp:double=0
    LinVelScale_UpDwn_Clamp:double=0.7
    DeltaTime:double=Get DeltaTimeLocal() -> (DeltaTimeLocal:double)
    GameTimeInSeconds:double=GetGameTimeInSeconds()
    WindGustsParams:Wind Gust Params Struct Structure=Get Wind Gust Params() -> (Wind Gust Params:Wind Gust Params Struct Structure)
    WindGustLocalSettings:Wind Gust Local Setting Struct Structure=Get Gust Local Settings() -> (Gust Local Settings:Wind Gust Local Setting Struct Structure)
   ) -> (
    then
    CurrentStateOut:String
    Cloth Settings Out:Cloth Setting Struct Structure
    Cloth Core Settings Out:Cloth Core Setting Struct Structure
    WindGustSettings Out:Wind Gust Setting Struct Structure
    WindGustLocalSettings Out:Wind Gust Local Setting Struct Structure
    GustStrength - WindVelocityOut:Vector
    LinearVelVector-DELETE LATER:Vector
   )
   > Set Gust Local Settings(Gust Local Settings:Wind Gust Local Setting Struct Structure=@n1.WindGustLocalSettings Out) -> (Output_Get:Wind Gust Local Setting Struct Structure)
   n2 = Event BlueprintUpdateAnimation()
   > Set DeltaTimeLocal(DeltaTimeLocal:double=@n2.DeltaTimeX) -> (Output_Get:double)
   > n3 = Sequence!(execute) -> (then_0, then_1, then_2:Exec, then_3:Exec, then_4:Exec, then_5:Exec, then_6:Exec, then_7:Exec, then_8:Exec, then_9:Exec)
   > Set Body Mesh(Body Mesh:Skeletal Mesh Component Object Reference=GetPartSkeletalMeshComponent(self=Self-Reference())) -> (Output_Get:Skeletal Mesh Component Object Reference)
   n4 = Gravity Override State Function DEPRECATED(
    execute=@n3.then_1
    CharacterPartAnimInstance=Self-Reference()
    GameTimeInSeconds=GetGameTimeInSeconds()
    GravityOverrideParams=Get Gravity Override Params()
    RBAN Params=Get RBAN Params()
   )
   > Set Gravity Override Cape(Gravity Override Cape:Vector=@n4) -> (Output_Get:Vector)
   > Set RBAN Settings Out(RBAN Settings Out:RBANSetting Struct Structure=@n4.RBAN SettingsOut) -> (Output_Get:RBANSetting Struct Structure)
   n5 = Get bIsSkyDiving()
   n6 = "To String (Object)"!(InObj:Object Reference=GetCurrentWeapon!(self:Fort Pawn Object Reference=Get As Fort Player Pawn() -> (As Fort Player Pawn:Fort Player Pawn Object Reference)) -> (ReturnValue:Fort Weapon Object Reference)) -> (ReturnValue:String)
   n7 = Branch(execute=@n3.then_2, Condition=AND Boolean(A=Contains(SearchIn=@n6, Substring=B_Athena_Pickaxe_DualWield_Generic), B=NOT Boolean(A=@n5)))
   > Set arms(arms:bool=false) -> (Output_Get:bool)
   Set arms(execute=@n7.else, arms=true)
   n8 = Branch(execute=@n3.then_3, Condition=CheckForBackbling())
   > Set Is Wearing Backbling(Is Wearing Backbling=true)
   Set Is Wearing Backbling(execute=@n8.else, Is Wearing Backbling=false)
   > Set Is Wearing Backbling(Is Wearing Backbling=Get bIsDBNOCarrying() -> (bIsDBNOCarrying:bool))
   Sequence!(execute=@n3.then_4) -> (then_0, then_1, then_2, then_3)
   > n9 = Vehicle Dynamics Scaler Function!(execute, CharacterPartAnimInstance:Custom Character Part Anim Instance Object Reference=Self-Reference(), Debug:bool=false) -> (then, InVehicleHeadScaleAlpha:double, InVehicleHeadSimAlpha:double, InVehicleSkirtScaleAlpha:double, InVehicleSkirtSimAlpha:double)
   > Set in vehicle scale alpha(in vehicle scale alpha:double=@n9.InVehicleSkirtScaleAlpha) -> (Output_Get:double)
   > Set in vehicle scale alpha(in vehicle scale alpha=@n9.InVehicleSkirtSimAlpha)
   Set FrontEnd or Emote(execute=@n3.then_7, FrontEnd or Emote:bool=OR Boolean(A=Get bIsInFrontEnd(), B=Get bIsPlayingEmote())) -> (Output_Get:bool)
   n10 = NOT Boolean(A=Get bIsInFrontEnd())
   n11 = Get bIsPlayingEmote()
   > Set Crouching or Emote or Siding(Crouching or Emote or Siding:bool=OR Boolean!(A=AND Boolean(A=@n11, B=@n10), B=Get bIsCrouching(), C:bool=Get bIsSliding()) -> (ReturnValue)) -> (Output_Get:bool)
   > Set Not Wearing Backbling(Not Wearing Backbling:bool=NOT Boolean(A=Get Is Wearing Backbling())) -> (Output_Get:bool)
   n12 = Montage_IsPlaying(
    self=GetAnimInstance(self=GetOwnerSkeletalMeshComponent())
    Montage=/BRCosmetics/Animation/Game/MainPlayer/Emotes/SherWolf/CMM/Emote_Sherwolf_CMM_Hero_M
   )
   n13 = Branch(execute=@n3.then_8, Condition=@n12)
   n14 = SelectFloat(A=1, B=0, bPickA=@n12)
   > Set Is Playing Custom Emote(Is Playing Custom Emote:double=@n14) -> (Output_Get:double)
   > Set emote_sherwolf(emote_sherwolf:bool=true) -> (Output_Get:bool)
   Set Is Playing Custom Emote(execute=@n13.else, Is Playing Custom Emote=@n14)
   > Set emote_sherwolf(emote_sherwolf=false)
   n15 = Branch(execute=@n3.then_6, Condition=Get bRigidBodyEnabled())
   > Set TrailControllerRot(TrailControllerRot:Rotator="0, 0, 0") -> (Output_Get:Rotator)
   > Set BwdMovement(BwdMovement:double=0) -> (Output_Get:double)
   > Set reposeCapeTrail(reposeCapeTrail:bool=false) -> (Output_Get:bool)
   > Set CRAlpha(CRAlpha:double=1) -> (Output_Get:double)
   n16 = Sequence!(execute=@n15.else) -> (then_0, then_1)
   > Set reposeCapeTrail(reposeCapeTrail=true)
   > Set CRAlpha(CRAlpha=0)
   > n17 = PlayerMovement(AnimInstance=Self-Reference(), XDotProductCoefficient=1, YDotProductCoefficient=1, ZDotProductCoefficient=0)
   n18 = MapRangeClamped(Value=@n17.YMovement, InRangeA=0, InRangeB=1)
   n19 = "Absolute (Float)"(A=@n17)
   > Set BwdMovement(BwdMovement=MapRangeClamped(Value=float + float(A=@n19, B=@n18) {is=title K2Node_CommutativeAssociativeBinaryOperator}, InRangeA=0, InRangeB=1.4))
   > n20 = Branch(Condition=Get bIsSwimSprinting() -> (bIsSwimSprinting:bool))
   > Set TrailControllerRot(TrailControllerRot="0,-15, 0")
   n21 = Branch(execute=@n20.else, Condition=Get bIsSkyDiving())
   > n22 = Branch(Condition=Get bIsParachuteOpened())
   > Set TrailControllerRot(TrailControllerRot="0,-15,0")
   > Set BwdMovement(BwdMovement=0)
   n23 = Branch(execute=@n22.else, Condition="float > float"(A=Get PawnVelocityZ(), B=50) {is=title K2Node_CallFunction})
   > Set BwdMovement(BwdMovement=float * float(A=Get BwdMovement(), B=0))
   > Set TrailControllerRot(TrailControllerRot="0,-25, 0")
   n24 = Branch(execute=@n23.else, Condition=Get bIsSkydiveDiveMode())
   > Set TrailControllerRot(TrailControllerRot="0,-15,0")
   Set TrailControllerRot(execute=@n24.else, TrailControllerRot="0,-25,0")
   Set TrailControllerRot(execute=@n21.else, TrailControllerRot="0,-12.5, 0")
   > Set BwdMovement(BwdMovement=float * float(A=Get BwdMovement(), B=1))
   Set skydive(execute=@n16.then_1, skydive:bool=OR Boolean(A=Get bIsSkyDiving(), B=Get bIsSkydiveFloating(), C=Get bIsSkydiveDiveMode())) -> (Output_Get:bool)
   > n25 = Branch(Condition=Get skydive() -> (skydive:bool))
   > Set clavicle_float_r(clavicle_float_r:double=0) -> (Output_Get:double)
   > Set clavicle_float_l(clavicle_float_l:double=0) -> (Output_Get:double)
   n26 = GetDeltaTransformFromRefPose(self=GetPartSkeletalMeshComponent(), BoneName=upperarm_l, BaseName=spine_05)
   n27 = MapRangeClamped(Value=BreakRotator(InRot=BreakTransform(InTransform=@n26).Rotation).Yaw, InRangeA=10, InRangeB=-30)
   > Set clavicle_glider_float_l(clavicle_glider_float_l:double=@n27) -> (Output_Get:double)
   n28 = BreakTransform(InTransform=GetDeltaTransformFromRefPose(self=GetPartSkeletalMeshComponent(), BoneName=upperarm_r, BaseName=spine_05))
   Set clavicle_float_r(execute=@n25.else, clavicle_float_r=MapRangeClamped(Value=BreakRotator(InRot=@n28.Rotation).Yaw, InRangeA=0, InRangeB=-40))
   > Set clavicle_float_l(clavicle_float_l=@n27)
   > Set clavicle_glider_float_l(clavicle_glider_float_l=0)
   n29 = Get bIsPlayingEmote()
   n30 = Get bRigidBodyEnabled()
   n31 = Branch(
    execute=@n3.then_9
    Condition=AND Boolean!(A=NOT Boolean(A=@n29), B=NOT Boolean(A=@n30), C:bool=Get NotClothEnabled(), D:bool=Get bIsInFrontEnd()) -> (ReturnValue)
   )
   > Set Is_Emoting_FE(Is_Emoting_FE:bool=true) -> (Output_Get:bool)
   Set Is_Emoting_FE(execute=@n31.else, Is_Emoting_FE=false)
   Event BlueprintInitializeAnimation()
   n32 = GetOwnerSkeletalMeshComponent()
   > Is Valid(InputObject=@n32)
   > Set Skel_Mesh(Skel_Mesh=@n32)
   > n33 = Cast To FortPlayerPawn(Object=GetOwningActor())
   > n34 = Set Pawn(Pawn:Fort Player Pawn Object Reference=@n33) -> (Output_Get:Fort Player Pawn Object Reference)
   > Set Anim_Inst(Anim_Inst:Anim Instance Object Reference=GetAnimInstance(self=Get Mesh(self=@n34))) -> (Output_Get:Anim Instance Object Reference)
   > n35 = Cast To FortPlayerPawn(Object=TryGetPawnOwner())
   > Set As Fort Player Pawn(As Fort Player Pawn:Fort Player Pawn Object Reference=@n35) -> (Output_Get:Fort Player Pawn Object Reference)
   K2Node_Event_4 = CustomEvent() {is=title K2Node_CustomEvent}
   > Set TrailControllerRot(TrailControllerRot="0,-65,0")
   > Delay(Duration=0.3)
   > ResetDynamics(InTeleportType=TeleportPhysics)
   > ResetDynamics(InTeleportType=ResetPhysics)
   K2Node_Event_6 = CustomEvent_1() {is=title K2Node_CustomEvent}
   > Set TrailControllerRot(TrailControllerRot="0,-10,0")
   DrawVectorFromJointFunction!(
    execute
    SkelMeshComponent:Scene Component Object Reference
    JointName:Name=pelvis
    LineEnd:Vector=Break Cloth Setting Struct!(ClothSettingStruct:"Cloth Setting Struct Structure (by ref)"=@n1.Cloth Settings Out) -> (
     LinearVelocityScale_7_#1:Vector=~
     AngularVelocityScale_6_#2:float=~
     FictitiousAngularScale_10_#3:float=~
     GravityOverride_15_#4:Vector=~
     DragCoefficient_23_#5:Vector 2D Structure
     LiftCoefficient_24_#6:Vector 2D Structure
     JointName_27_#7:Name
    ) {is=title K2Node_BreakStruct}.GravityOverride_15_#4
    JointOffset:Vector="0, 0, 0"
    LineColor:Linear Color Structure=(R=1,G=0,B=0,A=1)
    Thickness:double=2
    Foreground:EDrawDebugSceneDepthPriorityGroup Enum=Foreground
   ) -> (then)
   Get bIsPlayingEmote()
  ]}
]
variables [
 Anim_Inst {type=AnimInstance, default=null}
 arms {type=boolean}
 As Fort Player Pawn {type=FortPlayerPawn}
 As Fort Player Pawn_0 {type=FortPlayerPawn}
 Body Mesh {type=SkeletalMeshComponent}
 BwdMovement {type=number}
 clavicle_float_l {type=number}
 clavicle_float_r {type=number}
 clavicle_glider_float_l {type=number}
 Cloth Core Settings_0 {
  type=ClothCoreSettingStruct
  default {
   collisionThickness=1
   dampingCoefficient=0.01
   frictionCoefficient=0.5
   localDampingCoefficient=0
   numberIterations=2
   numberSubsteps=1
   selfCollisionThickness=0
   animDriveDamping=<0 1>
   animDriveStiffness=<0 1>
   areaStiffness=<1 1>
   bendingStiffness=<1 1>
   edgeStiffness=<1 1>
   tetherStiffness=<1 1>
  }
 }
 Cloth Paramaters_0 {
  type=ClothParamsStruct
  default {
   dBNO {
    angularVelocityScale=0.6
    fictitiousAngularScale=0.1
    jointName=pelvis
    dragCoefficient=<0.2 1>
    gravityOverride=<-980 400 0>
    liftCoefficient=<0 0.5>
    linearVelocityScale=<0.3 0.3 0.1>
   }
   dBNO_Carried {
    angularVelocityScale=0.6
    fictitiousAngularScale=0.1
    jointName=pelvis
    dragCoefficient=<0.2 1>
    gravityOverride=<-980 600 0>
    liftCoefficient=<0 0.5>
    linearVelocityScale=<0.3 0.3 0.1>
   }
   dBNO_Impulse {
    angularVelocityScale=0
    fictitiousAngularScale=0
    jointName=pelvis
    dragCoefficient=<0 0.2>
    gravityOverride=<-3000 -2000 0>
    liftCoefficient=<0 0.1>
    linearVelocityScale=<0.3 0.1 0.3>
   }
   dBNO_RollBwd {
    angularVelocityScale=0
    fictitiousAngularScale=0
    jointName=pelvis
    dragCoefficient=<0 0.2>
    gravityOverride=<-2000 0 0>
    liftCoefficient=<0 0.1>
    linearVelocityScale=<0.3 0.1 0.3>
   }
   dBNO_RollFwd {
    angularVelocityScale=0
    fictitiousAngularScale=0
    jointName=pelvis
    dragCoefficient=<0 0.2>
    gravityOverride=<-980 0 0>
    liftCoefficient=<0 0.1>
    linearVelocityScale=<0.3 0.1 0.3>
   }
   dBNO_Tumbling {
    angularVelocityScale=0
    fictitiousAngularScale=0
    jointName=pelvis
    dragCoefficient=<0 0.2>
    gravityOverride=<-2500 -300 0>
    liftCoefficient=<0 0.1>
    linearVelocityScale=<0.3 0.1 0.3>
   }
   emoteOrMelee {
    angularVelocityScale=0.3
    fictitiousAngularScale=1
    jointName=root
    dragCoefficient=<1 1>
    gravityOverride=<0 0 -980>
    liftCoefficient=<0.1 0.15>
    linearVelocityScale=<0.3 0.3 0.3>
   }
   front_End {
    angularVelocityScale=0.1
    fictitiousAngularScale=0.3
    jointName=pelvis
    dragCoefficient=<0.3 0.5>
    gravityOverride=<-800 20 0>
    liftCoefficient=<0.5 0.5>
    linearVelocityScale=<0.2 0.2 0.3>
   }
   onGround_Crouch_Moving {
    angularVelocityScale=0.1
    fictitiousAngularScale=0.5
    jointName=root
    dragCoefficient=<0.5 0.5>
    gravityOverride=<0 -300 -700>
    liftCoefficient=<0.1 0.2>
    linearVelocityScale=<0.5 0.5 0.3>
   }
   onGround_Falling {
    angularVelocityScale=0.6
    fictitiousAngularScale=0
    jointName=root
    dragCoefficient=<0 0.2>
    gravityOverride=<0 -500 -980>
    liftCoefficient=<0 0.1>
    linearVelocityScale=<2 1 0.3>
   }
   onGround_Moving {
    angularVelocityScale=0.1
    fictitiousAngularScale=0.5
    jointName=root
    dragCoefficient=<0.3 1>
    gravityOverride=<0 0 -980>
    liftCoefficient=<0.1 0.3>
    linearVelocityScale=<0.5 0.5 0.3>
   }
   onGround_Standing {
    angularVelocityScale=0.1
    fictitiousAngularScale=0.3
    jointName=root
    dragCoefficient=<0.5 0.5>
    gravityOverride=<0 -150 -980>
    liftCoefficient=<0 0.5>
    linearVelocityScale=<0.3 0.3 0.3>
   }
   rail_Grinding {
    angularVelocityScale=0.2
    fictitiousAngularScale=0
    jointName=root
    dragCoefficient=<0.1 0.2>
    gravityOverride=<50 -500 -400>
    liftCoefficient=<0 0.5>
    linearVelocityScale=<0.1 0.1 0.1>
   }
   riding_Moving {
    angularVelocityScale=0.05
    fictitiousAngularScale=1
    jointName=pelvis
    dragCoefficient=<0.5 0.5>
    gravityOverride=<-980 0 0>
    liftCoefficient=<0.5 0.5>
    linearVelocityScale=<0.4 0.6 0.3>
   }
   riding_Moving_Targeting {
    angularVelocityScale=0.05
    fictitiousAngularScale=0.7
    jointName=root
    dragCoefficient=<0.5 0.5>
    gravityOverride=<-980 0 0>
    liftCoefficient=<0.5 0.5>
    linearVelocityScale=<0.2 0.6 0.2>
   }
   riding_Standing {
    angularVelocityScale=0.05
    fictitiousAngularScale=1
    jointName=pelvis
    dragCoefficient=<0.5 0.5>
    gravityOverride=<-980 -450 0>
    liftCoefficient=<0.5 0.5>
    linearVelocityScale=<0.6 0.6 0.3>
   }
   skydive_DivingDown {
    angularVelocityScale=0.1
    fictitiousAngularScale=5
    jointName=pelvis
    dragCoefficient=<0 0.1>
    gravityOverride=<-300 -150 0>
    liftCoefficient=<0 0>
    linearVelocityScale=<0.1 0.1 0.1>
   }
   skydive_FloatingDown {
    angularVelocityScale=0.1
    fictitiousAngularScale=0
    jointName=pelvis
    dragCoefficient=<0 0.1>
    gravityOverride=<-100 0 0>
    liftCoefficient=<0 0>
    linearVelocityScale=<0.1 0.1 0.1>
   }
   skydive_GoingUp {
    angularVelocityScale=0.1
    fictitiousAngularScale=0
    jointName=pelvis
    dragCoefficient=<0 0.1>
    gravityOverride=<100 -1200 0>
    liftCoefficient=<0 0>
    linearVelocityScale=<0.1 0.1 0.05>
   }
   skydive_Parachuting {
    angularVelocityScale=0.1
    fictitiousAngularScale=0
    jointName=pelvis
    dragCoefficient=<0.3 0.5>
    gravityOverride=<-980 -300 0>
    liftCoefficient=<0.1 0.5>
    linearVelocityScale=<0.1 0.1 0.1>
   }
   skydive_Surfing {
    angularVelocityScale=0.6
    fictitiousAngularScale=0.1
    jointName=root
    dragCoefficient=<1.5 1.5>
    gravityOverride=<0 -600 -500>
    liftCoefficient=<1 1>
    linearVelocityScale=<0.1 0.1 0.3>
   }
   swimming {
    angularVelocityScale=0.4
    fictitiousAngularScale=0
    jointName=root
    dragCoefficient=<0 0.5>
    gravityOverride=<0 0 -400>
    liftCoefficient=<0 0.1>
    linearVelocityScale=<0.4 0.2 0.2>
   }
   tactical_Sprint {
    angularVelocityScale=0.1
    fictitiousAngularScale=0.5
    jointName=root
    dragCoefficient=<0.3 1>
    gravityOverride=<0 -300 -980>
    liftCoefficient=<0.1 0.3>
    linearVelocityScale=<0.5 0.5 0.4>
   }
   ziplining {
    angularVelocityScale=0.6
    fictitiousAngularScale=0.2
    jointName=root
    dragCoefficient=<0 1.5>
    gravityOverride=<0 -200 -400>
    liftCoefficient=<0 0.3>
    linearVelocityScale=<0 0 0>
   }
  }
 }
 CRAlpha {type=number, default=1}
 Crouching or Emote or Siding {type=boolean}
 DeltaTimeLocal {type=number}
 emote_sherwolf {type=boolean}
 FrontEnd or Emote {type=boolean}
 Gravity Override Cape {type=Vector}
 Gravity Override Params {
  type=GravityOverrideParamsStruct
  default {
   emoteOrMelee {gravityOverride=<0 -150 -800>}
   onGround_Moving {gravityOverride=<0 -300 -800>}
   onGround_Standing {gravityOverride=<0 -150 -980>}
   skydive_GoingUp {gravityOverride=<0 -300 600>}
   tactical_Sprint {gravityOverride=<0 -450 0>}
  }
 }
 Gust Local Settings {type=WindGustLocalSettingStruct, default {elapsedTime=0, finalGustStrengthLocal=0, gustDuration=0, gustStrength=0, timeBetweenGusts=0}}
 in vehicle rigid alpha {type=number, default=1}
 in vehicle scale alpha {type=number}
 Is Playing Custom Emote {type=number}
 Is Wearing Backbling {type=boolean}
 Is_Emoting_FE {type=boolean}
 Not Wearing Backbling {type=boolean}
 NotClothEnabled {type=boolean, default=true}
 Pawn {type=FortPlayerPawn}
 RBAN Params {
  type=RBANParamsStruct
  default {
   dBNO {simSpaceSettings {
     maxAngularAcceleration=3
     maxAngularVelocity=5
     maxLinearAcceleration=500
     maxLinearVelocity=1500
     velocityScaleZ=1
     externalLinearDragV=<1 1 0>
    }}
   dBNO_Carried {simSpaceSettings {maxAngularAcceleration=1, maxAngularVelocity=1, maxLinearAcceleration=10, maxLinearVelocity=0, velocityScaleZ=1, externalLinearDragV=<1 1 0>}}
   emoteOrMelee {
    componentLinearAccScale=<1 1 1>
    componentLinearVelScale=<1.1 1.1 1.1>
    simSpaceSettings {
     maxAngularAcceleration=3
     maxAngularVelocity=5
     maxLinearAcceleration=500
     maxLinearVelocity=1500
     velocityScaleZ=1
     externalLinearDragV=<1 1 0>
    }
   }
   front_End {simSpaceSettings {
     maxAngularAcceleration=1
     maxAngularVelocity=3.5
     maxLinearAcceleration=10
     maxLinearVelocity=500
     velocityScaleZ=1
     externalLinearDragV=<1 1 0>
    }}
   onGround_Crouch_Moving {simSpaceSettings {
     maxAngularAcceleration=3
     maxAngularVelocity=5
     maxLinearAcceleration=500
     maxLinearVelocity=1500
     velocityScaleZ=1
     externalLinearDragV=<1 1 0>
    }}
   onGround_Falling {simSpaceSettings {
     maxAngularAcceleration=3
     maxAngularVelocity=5
     maxLinearAcceleration=500
     maxLinearVelocity=1500
     velocityScaleZ=1
     externalLinearDragV=<1 1 0>
    }}
   onGround_Moving {
    componentLinearAccScale=<0.1 0.1 0.1>
    componentLinearVelScale=<0.2 0.2 0.2>
    simSpaceSettings {
     maxAngularAcceleration=3
     maxAngularVelocity=5
     maxLinearAcceleration=500
     maxLinearVelocity=1500
     velocityScaleZ=1
     externalLinearDragV=<1 1 0>
    }
   }
   onGround_Standing {
    componentLinearAccScale=<0.2 0.2 0.3>
    componentLinearVelScale=<0.3 0.3 0.4>
    simSpaceSettings {
     maxAngularAcceleration=50
     maxAngularVelocity=5
     maxLinearAcceleration=500
     maxLinearVelocity=1500
     velocityScaleZ=1
     externalLinearDragV=<1 1 0>
    }
   }
   rail_Grinding {simSpaceSettings {
     maxAngularAcceleration=3
     maxAngularVelocity=5
     maxLinearAcceleration=500
     maxLinearVelocity=1500
     velocityScaleZ=1
     worldAlpha=1
     externalLinearDragV=<1 1 0>
    }}
   riding_Moving {simSpaceSettings {
     maxAngularAcceleration=3
     maxAngularVelocity=5
     maxLinearAcceleration=500
     maxLinearVelocity=1500
     velocityScaleZ=1
     externalLinearDragV=<1 1 0>
    }}
   riding_Moving_Targeting {simSpaceSettings {
     maxAngularAcceleration=3
     maxAngularVelocity=5
     maxLinearAcceleration=500
     maxLinearVelocity=1500
     velocityScaleZ=1
     externalLinearDragV=<1 1 0>
    }}
   riding_Standing {simSpaceSettings {
     maxAngularAcceleration=3
     maxAngularVelocity=5
     maxLinearAcceleration=500
     maxLinearVelocity=1500
     velocityScaleZ=1
     externalLinearDragV=<1 1 0>
    }}
   skydive_DivingDown {
    componentLinearAccScale=<0 0 0>
    componentLinearVelScale=<0 0 0>
    simSpaceSettings {
     maxAngularAcceleration=3
     maxAngularVelocity=5
     maxLinearAcceleration=500
     maxLinearVelocity=1500
     velocityScaleZ=1
     externalLinearDragV=<1 1 0>
    }
   }
   skydive_FloatingDown {
    componentLinearAccScale=<0 0 0>
    componentLinearVelScale=<0 0 0>
    simSpaceSettings {
     maxAngularAcceleration=3
     maxAngularVelocity=5
     maxLinearAcceleration=500
     maxLinearVelocity=1500
     velocityScaleZ=1
     externalLinearDragV=<1 1 0>
    }
   }
   skydive_GoingUp {simSpaceSettings {
     maxAngularAcceleration=3
     maxAngularVelocity=5
     maxLinearAcceleration=500
     maxLinearVelocity=1500
     velocityScaleZ=1
     externalLinearDragV=<1 1 0>
    }}
   skydive_Parachuting {
    componentLinearAccScale=<0 0 0>
    componentLinearVelScale=<0 0 0>
    simSpaceSettings {
     maxAngularAcceleration=3
     maxAngularVelocity=5
     maxLinearAcceleration=500
     maxLinearVelocity=1500
     velocityScaleZ=1
     externalLinearDragV=<1 1 0>
    }
   }
   skydive_Surfing {simSpaceSettings {
     maxAngularAcceleration=3
     maxAngularVelocity=5
     maxLinearAcceleration=500
     maxLinearVelocity=1500
     velocityScaleZ=1
     externalLinearDragV=<1 1 0>
    }}
   swimming {simSpaceSettings {
     maxAngularAcceleration=3
     maxAngularVelocity=5
     maxLinearAcceleration=500
     maxLinearVelocity=1500
     velocityScaleZ=1
     externalLinearDragV=<1 1 0>
    }}
   tactical_Sprint {
    componentLinearAccScale=<0.2 0.2 0.2>
    componentLinearVelScale=<0.2 0.2 0.2>
    simSpaceSettings {
     maxAngularAcceleration=3
     maxAngularVelocity=5
     maxLinearAcceleration=500
     maxLinearVelocity=1500
     velocityScaleZ=1
     externalLinearDragV=<1 1 0>
    }
   }
   ziplining {simSpaceSettings {
     maxAngularAcceleration=0.5
     maxAngularVelocity=1
     maxLinearAcceleration=50
     maxLinearVelocity=200
     velocityScaleZ=1
     externalLinearDragV=<0.5 0.5 0>
    }}
  }
 }
 RBAN Settings Out {
  type=RBANSettingStruct
  default {
   componentLinearAccScale=<0 0 0>
   componentLinearVelScale=<0 0 0>
   simSpaceSettings {
    dampingAlpha=1
    maxAngularAcceleration=1000
    maxAngularVelocity=3
    maxLinearAcceleration=300
    maxLinearVelocity=700
    velocityScaleZ=0
    worldAlpha=1
    externalAngularVelocity=<0 0 0>
    externalLinearDragV=<0 0.5 0>
    externalLinearVelocity=<0 0 0>
   }
  }
 }
 reposeCapeTrail {type=boolean}
 Skel_Mesh {type=SkeletalMeshComponent}
 skydive {type=boolean}
 skydiveNoParachute {type=boolean}
 TrailControllerRot {type=Rotator, default=<rot 0 -12.5 0>}
 Wind Gust Params {
  type=WindGustParamsStruct
  default {
   dBNO {
    maxDownInterpSpeed=5
    maxGustDuration=2
    maxGustStrength=7
    maxTimeBetweenGusts=7
    maxUpInterpSpeed=5
    minDownInterpSpeed=2
    minGustDuration=0.5
    minGustStrength=3
    minTimeBetweenGusts=0
    minUpInterpSpeed=2
    windTurbulenceScalarA=0.1
    windTurbulenceScalarB=1
    windAmplitude=<0 0 0>
    windFrequency=<0 0 0>
    windOffset=<0 0 0>
   }
   dBNO_Carried {
    maxDownInterpSpeed=5
    maxGustDuration=2
    maxGustStrength=7
    maxTimeBetweenGusts=7
    maxUpInterpSpeed=5
    minDownInterpSpeed=2
    minGustDuration=0.5
    minGustStrength=3
    minTimeBetweenGusts=0
    minUpInterpSpeed=2
    windTurbulenceScalarA=0.1
    windTurbulenceScalarB=1
    windAmplitude=<0 0 0>
    windFrequency=<0 0 0>
    windOffset=<0 0 0>
   }
   dBNO_Impulse {
    maxDownInterpSpeed=5
    maxGustDuration=2
    maxGustStrength=7
    maxTimeBetweenGusts=7
    maxUpInterpSpeed=5
    minDownInterpSpeed=2
    minGustDuration=0.5
    minGustStrength=3
    minTimeBetweenGusts=0
    minUpInterpSpeed=2
    windTurbulenceScalarA=0.1
    windTurbulenceScalarB=1
    windAmplitude=<0 0 0>
    windFrequency=<0 0 0>
    windOffset=<0 0 0>
   }
   dBNO_RollBwd {
    maxDownInterpSpeed=5
    maxGustDuration=2
    maxGustStrength=7
    maxTimeBetweenGusts=7
    maxUpInterpSpeed=5
    minDownInterpSpeed=2
    minGustDuration=0.5
    minGustStrength=3
    minTimeBetweenGusts=0
    minUpInterpSpeed=2
    windTurbulenceScalarA=0.1
    windTurbulenceScalarB=1
    windAmplitude=<0 0 0>
    windFrequency=<0 0 0>
    windOffset=<0 0 0>
   }
   dBNO_RollFwd {
    maxDownInterpSpeed=5
    maxGustDuration=2
    maxGustStrength=7
    maxTimeBetweenGusts=7
    maxUpInterpSpeed=5
    minDownInterpSpeed=2
    minGustDuration=0.5
    minGustStrength=3
    minTimeBetweenGusts=0
    minUpInterpSpeed=2
    windTurbulenceScalarA=0.1
    windTurbulenceScalarB=1
    windAmplitude=<0 0 0>
    windFrequency=<0 0 0>
    windOffset=<0 0 0>
   }
   dBNO_Tumbling {
    maxDownInterpSpeed=5
    maxGustDuration=2
    maxGustStrength=7
    maxTimeBetweenGusts=7
    maxUpInterpSpeed=5
    minDownInterpSpeed=2
    minGustDuration=0.5
    minGustStrength=3
    minTimeBetweenGusts=0
    minUpInterpSpeed=2
    windTurbulenceScalarA=0.1
    windTurbulenceScalarB=1
    windAmplitude=<0 0 0>
    windFrequency=<0 0 0>
    windOffset=<0 0 0>
   }
   emoteOrMelee {
    maxDownInterpSpeed=5
    maxGustDuration=8
    maxGustStrength=7
    maxTimeBetweenGusts=8
    maxUpInterpSpeed=5
    minDownInterpSpeed=2
    minGustDuration=0.2
    minGustStrength=3
    minTimeBetweenGusts=1
    minUpInterpSpeed=2
    windTurbulenceScalarA=0.1
    windTurbulenceScalarB=0.8
    windAmplitude=<10 10 20>
    windFrequency=<5 5 8>
    windOffset=<-50 -50 -280>
   }
   front_End {
    maxDownInterpSpeed=5
    maxGustDuration=8
    maxGustStrength=7
    maxTimeBetweenGusts=8
    maxUpInterpSpeed=5
    minDownInterpSpeed=2
    minGustDuration=0.2
    minGustStrength=3
    minTimeBetweenGusts=1
    minUpInterpSpeed=2
    windTurbulenceScalarA=0.1
    windTurbulenceScalarB=0.8
    windAmplitude=<20 20 20>
    windFrequency=<2 4 2>
    windOffset=<-50 -50 0>
   }
   onGround_Crouch_Moving {
    maxDownInterpSpeed=5
    maxGustDuration=2
    maxGustStrength=7
    maxTimeBetweenGusts=7
    maxUpInterpSpeed=5
    minDownInterpSpeed=2
    minGustDuration=0.5
    minGustStrength=3
    minTimeBetweenGusts=0
    minUpInterpSpeed=2
    windTurbulenceScalarA=0.1
    windTurbulenceScalarB=1
    windAmplitude=<100 25 75>
    windFrequency=<1.5 0.5 7.5>
    windOffset=<0 -100 -175>
   }
   onGround_Falling {
    maxDownInterpSpeed=5
    maxGustDuration=2
    maxGustStrength=7
    maxTimeBetweenGusts=7
    maxUpInterpSpeed=5
    minDownInterpSpeed=2
    minGustDuration=0.5
    minGustStrength=3
    minTimeBetweenGusts=0
    minUpInterpSpeed=2
    windTurbulenceScalarA=0.1
    windTurbulenceScalarB=1
    windAmplitude=<0 0 0>
    windFrequency=<0 0 0>
    windOffset=<0 0 0>
   }
   onGround_Moving {
    maxDownInterpSpeed=5
    maxGustDuration=8
    maxGustStrength=7
    maxTimeBetweenGusts=8
    maxUpInterpSpeed=5
    minDownInterpSpeed=2
    minGustDuration=0.2
    minGustStrength=3
    minTimeBetweenGusts=1
    minUpInterpSpeed=2
    windTurbulenceScalarA=0.1
    windTurbulenceScalarB=0.8
    windAmplitude=<10 20 15>
    windFrequency=<0.3 0.3 0.3>
    windOffset=<10 -50 50>
   }
   onGround_Standing {
    maxDownInterpSpeed=5
    maxGustDuration=8
    maxGustStrength=7
    maxTimeBetweenGusts=8
    maxUpInterpSpeed=5
    minDownInterpSpeed=2
    minGustDuration=0.2
    minGustStrength=2
    minTimeBetweenGusts=1
    minUpInterpSpeed=2
    windTurbulenceScalarA=0.1
    windTurbulenceScalarB=0.8
    windAmplitude=<0 0 0>
    windFrequency=<1.5 2.3 3.7>
    windOffset=<0 0 0>
   }
   rail_Grinding {
    maxDownInterpSpeed=5
    maxGustDuration=8
    maxGustStrength=7
    maxTimeBetweenGusts=8
    maxUpInterpSpeed=5
    minDownInterpSpeed=2
    minGustDuration=0.2
    minGustStrength=3
    minTimeBetweenGusts=1
    minUpInterpSpeed=2
    windTurbulenceScalarA=0.1
    windTurbulenceScalarB=0.8
    windAmplitude=<20 65 42>
    windFrequency=<5 6 8.6>
    windOffset=<50 -500 -400>
   }
   riding_Moving {
    maxDownInterpSpeed=5
    maxGustDuration=2
    maxGustStrength=7
    maxTimeBetweenGusts=7
    maxUpInterpSpeed=5
    minDownInterpSpeed=2
    minGustDuration=0.5
    minGustStrength=3
    minTimeBetweenGusts=0
    minUpInterpSpeed=2
    windTurbulenceScalarA=0.1
    windTurbulenceScalarB=1
    windAmplitude=<0 0 0>
    windFrequency=<0 0 0>
    windOffset=<0 0 0>
   }
   riding_Moving_Targeting {
    maxDownInterpSpeed=5
    maxGustDuration=2
    maxGustStrength=7
    maxTimeBetweenGusts=7
    maxUpInterpSpeed=5
    minDownInterpSpeed=2
    minGustDuration=0.5
    minGustStrength=3
    minTimeBetweenGusts=0
    minUpInterpSpeed=2
    windTurbulenceScalarA=0.1
    windTurbulenceScalarB=1
    windAmplitude=<0 0 0>
    windFrequency=<0 0 0>
    windOffset=<0 0 0>
   }
   riding_Standing {
    maxDownInterpSpeed=5
    maxGustDuration=2
    maxGustStrength=7
    maxTimeBetweenGusts=7
    maxUpInterpSpeed=5
    minDownInterpSpeed=2
    minGustDuration=0.5
    minGustStrength=3
    minTimeBetweenGusts=0
    minUpInterpSpeed=2
    windTurbulenceScalarA=0.5
    windTurbulenceScalarB=1
    windAmplitude=<0 0 0>
    windFrequency=<0 0 0>
    windOffset=<0 0 0>
   }
   skydive_DivingDown {
    maxDownInterpSpeed=5
    maxGustDuration=8
    maxGustStrength=7
    maxTimeBetweenGusts=8
    maxUpInterpSpeed=5
    minDownInterpSpeed=2
    minGustDuration=0.2
    minGustStrength=3
    minTimeBetweenGusts=1
    minUpInterpSpeed=2
    windTurbulenceScalarA=0.1
    windTurbulenceScalarB=4
    windAmplitude=<0 25 35>
    windFrequency=<0 4 10>
    windOffset=<-450 -100 100>
   }
   skydive_FloatingDown {
    maxDownInterpSpeed=5
    maxGustDuration=8
    maxGustStrength=1
    maxTimeBetweenGusts=8
    maxUpInterpSpeed=5
    minDownInterpSpeed=2
    minGustDuration=0.2
    minGustStrength=0
    minTimeBetweenGusts=1
    minUpInterpSpeed=2
    windTurbulenceScalarA=0.5
    windTurbulenceScalarB=1
    windAmplitude=<50 20 35>
    windFrequency=<8 2 3>
    windOffset=<0 -300 200>
   }
   skydive_GoingUp {
    maxDownInterpSpeed=5
    maxGustDuration=8
    maxGustStrength=7
    maxTimeBetweenGusts=8
    maxUpInterpSpeed=5
    minDownInterpSpeed=2
    minGustDuration=0.2
    minGustStrength=3
    minTimeBetweenGusts=1
    minUpInterpSpeed=2
    windTurbulenceScalarA=0.1
    windTurbulenceScalarB=0.8
    windAmplitude=<0 0 0>
    windFrequency=<0 0 0>
    windOffset=<0 -500 300>
   }
   skydive_Parachuting {
    maxDownInterpSpeed=5
    maxGustDuration=8
    maxGustStrength=1
    maxTimeBetweenGusts=8
    maxUpInterpSpeed=5
    minDownInterpSpeed=2
    minGustDuration=5
    minGustStrength=0
    minTimeBetweenGusts=5
    minUpInterpSpeed=2
    windTurbulenceScalarA=0.2
    windTurbulenceScalarB=1
    windAmplitude=<25 50 75>
    windFrequency=<8 6 2>
    windOffset=<-200 -300 0>
   }
   skydive_Surfing {
    maxDownInterpSpeed=5
    maxGustDuration=2
    maxGustStrength=7
    maxTimeBetweenGusts=7
    maxUpInterpSpeed=5
    minDownInterpSpeed=2
    minGustDuration=0.5
    minGustStrength=3
    minTimeBetweenGusts=0
    minUpInterpSpeed=2
    windTurbulenceScalarA=0.5
    windTurbulenceScalarB=1
    windAmplitude=<20 30 25>
    windFrequency=<0.3 0.3 0.5>
    windOffset=<-150 -250 150>
   }
   swimming {
    maxDownInterpSpeed=5
    maxGustDuration=2
    maxGustStrength=7
    maxTimeBetweenGusts=7
    maxUpInterpSpeed=5
    minDownInterpSpeed=2
    minGustDuration=0.5
    minGustStrength=3
    minTimeBetweenGusts=0
    minUpInterpSpeed=2
    windTurbulenceScalarA=0.1
    windTurbulenceScalarB=1
    windAmplitude=<10 25 20>
    windFrequency=<1.5 0.5 7.5>
    windOffset=<0 -150 50>
   }
   tactical_Sprint {
    maxDownInterpSpeed=5
    maxGustDuration=8
    maxGustStrength=7
    maxTimeBetweenGusts=8
    maxUpInterpSpeed=5
    minDownInterpSpeed=2
    minGustDuration=0.2
    minGustStrength=3
    minTimeBetweenGusts=1
    minUpInterpSpeed=2
    windTurbulenceScalarA=0.1
    windTurbulenceScalarB=0.8
    windAmplitude=<10 20 15>
    windFrequency=<0.3 0.3 0.3>
    windOffset=<10 -50 50>
   }
   ziplining {
    maxDownInterpSpeed=5
    maxGustDuration=2
    maxGustStrength=7
    maxTimeBetweenGusts=7
    maxUpInterpSpeed=5
    minDownInterpSpeed=2
    minGustDuration=0.5
    minGustStrength=3
    minTimeBetweenGusts=0
    minUpInterpSpeed=2
    windTurbulenceScalarA=0.5
    windTurbulenceScalarB=1
    windAmplitude=<20 30 25>
    windFrequency=<0.3 0.3 0.5>
    windOffset=<-150 -250 150>
   }
  }
 }
 Z Movement {type=number}
]
